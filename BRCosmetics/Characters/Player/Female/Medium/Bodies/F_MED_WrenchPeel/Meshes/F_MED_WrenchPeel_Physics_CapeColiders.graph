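# Physics asset for the F_MED_WrenchPeel female body, named as the collider set for its cape.
# Single capsules cover the pelvis, spine_01–05, upper arms, thighs, calves and feet, leaving head, forearms and hands without collision.
# Constraints chain the bodies along the skeleton from the pelvis, with upper arms linked directly to spine_05.
# All swing and twist axes are free, so the constraints impose no angular limits.
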PhysicsAsset /BRCosmetics/Characters/Player/Female/Medium/Bodies/F_MED_WrenchPeel/Meshes/F_MED_WrenchPeel_Physics_CapeColiders
bodies {
 calf_l [{
   shapeType=Capsule
   length=38.7316
   radius=9.0727
   shapeName=calf_l_capsule
   center=<-24.3898 0.0001 0.2671>
   rotation=<rot -89.4065 -0.0014 0.0011>
  }]
 calf_r [{
   shapeType=Capsule
   length=38.7316
   radius=9.0727
   shapeName=calf_r_capsule
   center=<24.3898 -0.0001 -0.2671>
   rotation=<rot -89.4065 -0.0014 -179.9989>
  }]
 foot_l [{
   shapeType=Capsule
   length=13.9972
   radius=10
   shapeName=foot_l_capsule
   center=<0.0026 7.1941 0.0948>
   rotation=<rot -0.0019 -0.021 -90.7554>
  }]
 foot_r [{
   shapeType=Capsule
   length=13.9972
   radius=10
   shapeName=foot_r_capsule
   center=<-0.0026 -7.1941 -0.0948>
   rotation=<rot -0.0019 -0.021 89.2446>
  }]
 pelvis [{shapeType=Capsule, length=6.8951, radius=16.3153, shapeName=pelvis_capsule, center=<-1.329 -3.858 0>}]
 spine_01 [{shapeType=Capsule, length=7.9773, radius=15.2331, shapeName=pelvis_capsule, center=<2.0503 -3.5042 0>}]
 spine_02 [{shapeType=Capsule, length=9.9992, radius=13.2112, shapeName=pelvis_capsule, center=<2.302 -4.4625 0>}]
 spine_03 [{shapeType=Capsule, length=9.9992, radius=13.2112, shapeName=pelvis_capsule, center=<2.5504 -2.3792 0>}]
 spine_04 [{shapeType=Capsule, length=10.567, radius=12.6434, shapeName=pelvis_capsule, center=<2.5174 0.8898 0>}]
 spine_05 [{shapeType=Capsule, length=10.567, radius=12.6434, shapeName=pelvis_capsule, center=<1.8875 1.6908 0>}]
 thigh_l [{
   shapeType=Capsule
   length=28.7488
   radius=12.2405
   shapeName=thigh_l_capsule
   center=<-19.2481 0.0001 0.4724>
   rotation=<rot -89.3016 -0.0005 0.0003>
  }]
 thigh_r [{
   shapeType=Capsule
   length=28.7488
   radius=12.2405
   shapeName=thigh_r_capsule
   center=<19.2481 -0.0001 -0.4724>
   rotation=<rot -89.3016 -0.0005 -179.9997>
  }]
 upperarm_l [{
   shapeType=Capsule
   length=23.1859
   radius=6.5081
   shapeName=thigh_l_capsule
   center=<15.1568 -0.5815 0.1848>
   rotation=<rot -89.3016 -0.0005 0.0003>
  }]
 upperarm_r [{
   shapeType=Capsule
   length=23.1859
   radius=6.5081
   shapeName=thigh_r_capsule
   center=<-15.1568 0.5815 -0.1848>
   rotation=<rot -89.3016 -0.0005 -179.9997>
  }]
}
constraints [
 {
  bone1Name=spine_01
  bone2Name=pelvis
  swing1LimitDegrees=45
  swing1Motion=Free
  swing2LimitDegrees=45
  swing2Motion=Free
  twistLimitDegrees=45
  twistMotion=Free
 }
 {
  bone1Name=spine_02
  bone2Name=spine_01
  swing1LimitDegrees=45
  swing1Motion=Free
  swing2LimitDegrees=45
  swing2Motion=Free
  twistLimitDegrees=45
  twistMotion=Free
 }
 {
  bone1Name=spine_03
  bone2Name=spine_02
  swing1LimitDegrees=45
  swing1Motion=Free
  swing2LimitDegrees=45
  swing2Motion=Free
  twistLimitDegrees=45
  twistMotion=Free
 }
 {
  bone1Name=spine_04
  bone2Name=spine_03
  swing1LimitDegrees=45
  swing1Motion=Free
  swing2LimitDegrees=45
  swing2Motion=Free
  twistLimitDegrees=45
  twistMotion=Free
 }
 {
  bone1Name=spine_05
  bone2Name=spine_04
  swing1LimitDegrees=45
  swing1Motion=Free
  swing2LimitDegrees=45
  swing2Motion=Free
  twistLimitDegrees=45
  twistMotion=Free
 }
 {
  bone1Name=thigh_l
  bone2Name=pelvis
  swing1LimitDegrees=45
  swing1Motion=Free
  swing2LimitDegrees=45
  swing2Motion=Free
  twistLimitDegrees=45
  twistMotion=Free
 }
 {
  bone1Name=thigh_r
  bone2Name=pelvis
  swing1LimitDegrees=45
  swing1Motion=Free
  swing2LimitDegrees=45
  swing2Motion=Free
  twistLimitDegrees=45
  twistMotion=Free
 }
 {
  bone1Name=calf_r
  bone2Name=thigh_r
  swing1LimitDegrees=45
  swing1Motion=Free
  swing2LimitDegrees=45
  swing2Motion=Free
  twistLimitDegrees=45
  twistMotion=Free
 }
 {
  bone1Name=calf_l
  bone2Name=thigh_l
  swing1LimitDegrees=45
  swing1Motion=Free
  swing2LimitDegrees=45
  swing2Motion=Free
  twistLimitDegrees=45
  twistMotion=Free
 }
 {
  bone1Name=foot_r
  bone2Name=calf_r
  swing1LimitDegrees=45
  swing1Motion=Free
  swing2LimitDegrees=45
  swing2Motion=Free
  twistLimitDegrees=45
  twistMotion=Free
 }
 {
  bone1Name=foot_l
  bone2Name=calf_l
  swing1LimitDegrees=45
  swing1Motion=Free
  swing2LimitDegrees=45
  swing2Motion=Free
  twistLimitDegrees=45
  twistMotion=Free
 }
 {
  bone1Name=upperarm_l
  bone2Name=spine_05
  swing1LimitDegrees=45
  swing1Motion=Free
  swing2LimitDegrees=45
  swing2Motion=Free
  twistLimitDegrees=45
  twistMotion=Free
 }
 {
  bone1Name=upperarm_r
  bone2Name=spine_05
  swing1LimitDegrees=45
  swing1Motion=Free
  swing2LimitDegrees=45
  swing2Motion=Free
  twistLimitDegrees=45
  twistMotion=Free
 }
]
