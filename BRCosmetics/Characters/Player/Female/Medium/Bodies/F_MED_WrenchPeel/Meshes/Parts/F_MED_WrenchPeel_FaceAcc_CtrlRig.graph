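# Control Rig for the WrenchPeel character's face accessory part, procedurally correcting hair bones.
# Head rotation relative to spine_05, converted to Euler, drives the hair bunch base's rotation, translation and X scale through scaled, clamped and remapped values.
# A fixed offset adjusts deform_hair_back, while spherical pose readers on each upper arm, with debug drawing enabled, weight the left/right ALT back-hair deformers.
# An offset node for back-hair bases is left unconnected.
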ControlRigBlueprint /BRCosmetics/Characters/Player/Female/Medium/Bodies/F_MED_WrenchPeel/Meshes/Parts/F_MED_WrenchPeel_FaceAcc_CtrlRig
graphs [
 Multiply_2_ContainedGraph {nodes [
   n1 = RigVMFunctionEntryNode()
   RigVMFunctionReturnNode(Result=&RigVMFunction_MathDoubleMul(B=@n1.C, A=&RigVMFunction_MathDoubleMul(A=@n1.A, B=@n1.B).Result).Result)
  ]}
 Multiply_2_ContainedGraph {nodes [
   n1 = RigVMFunctionEntryNode()
   Multiply = &RigVMFunction_MathDoubleMul(A=@n1.A, B=@n1.B)
   RigVMFunctionReturnNode(Result=&RigVMFunction_MathDoubleMul(B=@n1.C, A=@Multiply.Result).Result)
  ]}
 RigVMFunctionLibrary {links [], nodes []}
 RigVMFunction_Sequence_1_ContainedGraph {nodes [
   n1 = RigVMFunctionEntryNode()
   RigVMFunction_Sequence = &RigVMFunction_Sequence(ExecuteContext=@n1.ExecuteContext)
   n2 = &RigVMFunction_Sequence(ExecuteContext=@RigVMFunction_Sequence.B)
   RigVMFunctionReturnNode(A=@RigVMFunction_Sequence.A, B=@n2.A, C=@n2.B)
  ]}
 RigVMModel {nodes [
   RigUnit_BeginExecution = &RigUnit_BeginExecution()
   RigVMFunction_Sequence = RigVMAggregateNode(ExecuteContext=@RigUnit_BeginExecution.ExecutePin)
   &RigUnit_ModifyTransforms(
    ItemToModify=((Item=(Type=Bone,Name="head"),Transform=(Rotation=(X=0,Y=0,Z=0,W=1),Translation=(X=0,Y=0,Z=0),Scale3D=(X=1,Y=1,Z=1))))
    Weight=1
    WeightMinimum=0
    ExecutePin=@RigVMFunction_Sequence.A
   )
   RigVMFunction_MathQuaternionToEuler = &RigVMFunction_MathQuaternionToEuler(
    RotationOrder=ZYX
    Value=&RigUnit_GetRelativeTransformForItem(Child=(Type=Bone,Name="head"), Parent=(Type=Bone,Name="spine_05"), bChildInitial=False, bParentInitial=False).RelativeTransform.Rotation
   )
   Clamp = &RigVMFunction_MathDoubleClamp(Maximum=10, Value=&RigVMFunction_MathDoubleMul(B=-0.02, A=@RigVMFunction_MathQuaternionToEuler.Result.Z).Result)
   Add = &RigVMFunction_MathDoubleAdd(B=-4.12, A=&RigVMFunction_MathDoubleMul(B=1.03, A=@RigVMFunction_MathQuaternionToEuler.Result.Z).Result)
   n1 = &RigVMFunction_MathDoubleMul(B=0.0011, A=@Add.Result)
   n2 = &RigVMFunction_MathDoubleRemap(SourceMaximum=-55, SourceMinimum=-10, TargetMaximum=1, TargetMinimum=0, Value=@RigVMFunction_MathQuaternionToEuler.Result.Z)
   Multiply = RigVMAggregateNode(B=0.5, C=@n2.Result, A=@RigVMFunction_MathQuaternionToEuler.Result.X)
   Multiply_1 = RigVMAggregateNode(B=0, C=@n2.Result, A=@RigVMFunction_MathQuaternionToEuler.Result.Y)
   RigVMFunction_MathQuaternionFromEuler = &RigVMFunction_MathQuaternionFromEuler(Euler=(X=0,Y=0,Z=15), Euler.Y=@Add.Result, Euler.X=@Multiply.Result, Euler.Z=@Multiply_1.Result)
   > &RigUnit_ModifyTransforms(
    ItemToModify=((Item=(Type=Bone,Name="dyn_hair_bunch_base"),Transform=(Rotation=(X=0,Y=-0.7071,Z=0,W=0.7071),Translation=(X=0,Y=1,Z=0),Scale3D=(X=0.9,Y=1,Z=1))))
    Weight=1
    WeightMinimum=0
    ItemToModify.0.Transform.Translation.X=@Clamp.Result
    ItemToModify.0.Transform.Translation.Z=&RigVMFunction_MathDoubleClamp(Maximum=10, Value=&RigVMFunction_MathDoubleMul(B=-0.03, A=@RigVMFunction_MathQuaternionToEuler.Result.Z).Result).Result
    ItemToModify.0.Transform.Scale3D.X=&RigVMFunction_MathDoubleClamp(Minimum=0.9, Value=&RigVMFunction_MathDoubleAdd(A=1, B=@n1.Result).Result).Result
    ItemToModify.0.Transform.Rotation=@RigVMFunction_MathQuaternionFromEuler.Result
   )
   &RigUnit_ModifyTransforms(
    ItemToModify=((Item=(Type=Bone,Name="deform_hair_back"),Transform=(Rotation=(X=0,Y=0,Z=0,W=1),Translation=(X=0,Y=0,Z=2),Scale3D=(X=1,Y=1,Z=1))))
    Weight=1
    WeightMinimum=0
    ExecutePin=@RigVMFunction_Sequence.B
   )
   n3 = &RigVMFunction_Sequence(ExecuteContext=@RigVMFunction_Sequence.C)
   n4 = &RigUnit_SphericalPoseReader(
    ActiveRegionScaleFactors=(PositiveWidth=1,NegativeWidth=1,PositiveHeight=1,NegativeHeight=1)
    ActiveRegionSize=0.05
    Debug=(bDrawDebug=True,bDraw2D=False,bDrawLocalAxes=False,DebugScale=25,DebugSegments=20,DebugThickness=0.25)
    DriverAxis=(X=1,Y=0,Z=0)
    DriverItem=(Type=Bone,Name="upperarm_l")
    FalloffRegionScaleFactors=(PositiveWidth=1,NegativeWidth=1,PositiveHeight=1,NegativeHeight=1)
    FalloffSize=0.25
    FlipHeightScaling=False
    FlipWidthScaling=False
    OptionalParentItem=(Type=Bone,Name="spine_05")
    RotationOffset=(X=-125,Y=0,Z=60)
    ExecutePin=@n3.A
   )
   > &RigUnit_ModifyTransforms(
    ItemToModify=((Item=(Type=Bone,Name="deform_hair_back_ALT_l"),Transform=(Rotation=(X=0,Y=0,Z=-0.1305,W=0.9914),Translation=(X=0,Y=-9,Z=0),Scale3D=(X=1,Y=1,Z=1))))
    WeightMinimum=0
    Weight=@n4.OutputParam
   )
   n5 = &RigUnit_SphericalPoseReader(
    ActiveRegionScaleFactors=(PositiveWidth=1,NegativeWidth=1,PositiveHeight=1,NegativeHeight=1)
    ActiveRegionSize=0.05
    Debug=(bDrawDebug=True,bDraw2D=False,bDrawLocalAxes=False,DebugScale=25,DebugSegments=20,DebugThickness=0.25)
    DriverAxis=(X=-1,Y=0,Z=0)
    DriverItem=(Type=Bone,Name="upperarm_r")
    FalloffRegionScaleFactors=(PositiveWidth=1,NegativeWidth=1,PositiveHeight=1,NegativeHeight=1)
    FalloffSize=0.25
    FlipHeightScaling=False
    FlipWidthScaling=False
    OptionalParentItem=(Type=Bone,Name="spine_05")
    RotationOffset=(X=55,Y=0,Z=60)
    ExecutePin=@n3.B
   )
   > &RigUnit_ModifyTransforms(
    ItemToModify=((Item=(Type=Bone,Name="deform_hair_back_ALT_r"),Transform=(Rotation=(X=0,Y=0,Z=-0.1305,W=0.9914),Translation=(X=0,Y=9,Z=0),Scale3D=(X=1,Y=1,Z=1))))
    WeightMinimum=0
    Weight=@n5.OutputParam
   )
   n6 = RigVMDispatchNode(Elements=(X=1,Y=0,Z=5)) {template="DISPATCH_RigVMDispatch_MakeStruct(in Elements,out Struct)"}
   &RigUnit_ModifyTransforms(
    ExecutePin=()
    ItemToModify="((Item=(Type=Bone,Name=\"dyn_hair_back_01_base_l\"),Transform=(Rotation=(X=0,Y=0,Z=0,W=1),Translation=(X=0,Y=0,Z=10),Scale3D=(X=1,Y=1,Z=1))),(Item=(Type=Bone,Name=\"dyn_hair_back_01_base_r\"),Transform=(Rotation=(X=0,Y=0,Z=0,W=1),Translation=(X=0,Y=0,Z=10),Sca ...(truncated)"
    Weight=1
    WeightMinimum=0
    ItemToModify.0.Transform.Translation=@n6.Struct
    ItemToModify.1.Transform.Translation=@n6.Struct
    ItemToModify.2.Transform.Translation=@n6.Struct
    ItemToModify.3.Transform.Translation=@n6.Struct
   )
  ]}
]
hierarchy [
 root {loc=<0 0 0>, rot=<0 0 0 1>}
 pelvis {parent=root, loc=<0 -0.5301 98.0571>, rot=<0.0127 0.707 -0.0127 -0.707>}
 spine_01 {parent=pelvis, loc=<6.1208 -0.9328 0>, rot=<0 0 0.034 -0.9994>}
 spine_02 {parent=spine_01, loc=<6.5997 -0.2095 0>, rot=<0 0 0.0238 -0.9997>}
 spine_03 {parent=spine_02, loc=<6.4914 -0.1237 0>, rot=<0 0 0.0149 -0.9999>}
 spine_04 {parent=spine_03, loc=<6.9598 0.0794 0>, rot=<0 0 -0.0132 -0.9999>}
 spine_05 {parent=spine_04, loc=<6.9987 0 0>, rot=<0 0 -0.0015 -1>}
 neck_01 {parent=spine_05, loc=<11.3138 0.1529 0>, rot=<0 0 -0.2003 -0.9797>}
 neck_02 {parent=neck_01, loc=<4.5627 0 0>, rot=<0 0 0.0504 -0.9987>}
 head {parent=neck_02, loc=<5.3914 0 0>, rot=<0 0 0.115 -0.9934>}
 dyn_banana_base {parent=head, loc=<7.4497 6.0845 -5.8664>, rot=<-0.3894 -0.8873 -0.2461 -0.0211>}
 dyn_banana_01 {parent=dyn_banana_base, loc=<0 0 0>, rot=<0 0 0 1>}
 dyn_hair_base_fr_l {parent=head, loc=<4.8516 9.2514 -4.1858>, rot=<0.3311 -0.8535 -0.258 -0.3088>}
 dyn_hair_01_fr_l {parent=dyn_hair_base_fr_l, loc=<0 0 0>, rot=<0 0 0 1>}
 dyn_hair_base_fr_r {parent=head, loc=<4.5229 8.2219 5.4914>, rot=<0.2415 -0.0114 0.9695 0.0404>}
 dyn_hair_01_fr_r {parent=dyn_hair_base_fr_r, loc=<0 0 0>, rot=<0 0 0 1>}
 dyn_hair_base_side_r {parent=head, loc=<2.3021 4.3455 6.6717>, rot=<0.0188 -0.0021 0.9989 0.0419>}
 dyn_hair_01_side_r {parent=dyn_hair_base_side_r, loc=<0 0 0>, rot=<0 0 0 1>}
 dyn_hair_base_side_l {parent=head, loc=<3.1857 3.5751 -6.4418>, rot=<-0.0209 -0.9996 -0.0174 -0.0107>}
 dyn_hair_01_side_l {parent=dyn_hair_base_side_l, loc=<0 0 0>, rot=<0 0 0 1>}
 dyn_hair_bunch_base {parent=head, loc=<5.7985 -7.2362 -0.0109>, rot=<0.503 -0.4996 0.4973 -0.5001>}
 dyn_hair_bunch_01 {parent=dyn_hair_bunch_base, loc=<0 0 0>, rot=<0 0 0 1>}
 dyn_hair_tailbunch_mid {parent=dyn_hair_bunch_base, loc=<1.6241 0.1745 -2.6192>, rot=<0.0027 -0.3894 -0.0455 -0.9199>}
 dyn_hair_tailbunch_r {parent=dyn_hair_bunch_base, loc=<1.8378 -3.0693 -3.2675>, rot=<0.0682 -0.4415 0.1305 -0.8851>}
 dyn_hair_tailbunch_l {parent=dyn_hair_bunch_base, loc=<1.3205 3.5611 -3.0603>, rot=<0.2651 -0.4055 -0.3756 -0.7901>}
 dyn_hair_pigtail_base {parent=head, loc=<2.6008 -6.8845 -0.1954>, rot=<0.3007 -0.641 0.6441 -0.2897>}
 dyn_hair_pigtail_01 {parent=dyn_hair_pigtail_base, loc=<0 0 0>, rot=<0 0 0 1>}
 dyn_hair_pigtail_02 {parent=dyn_hair_pigtail_01, loc=<8.5586 0.1222 -1.8425>, rot=<0.0045 -0.1926 -0.0036 -0.9813>}
 dyn_hair_pigtail_03 {parent=dyn_hair_pigtail_02, loc=<8.5356 0.0335 -1.139>, rot=<0.0007 -0.1011 0.0066 -0.9949>}
 dyn_hair_pigtail_04 {parent=dyn_hair_pigtail_03, loc=<8.4813 0.0575 -0.1173>, rot=<-0.0018 -0.0599 -0.0428 -0.9973>}
 dyn_hair_pigtail_05 {parent=dyn_hair_pigtail_04, loc=<8.1479 0.1552 -0.3158>, rot=<-0.0014 -0.0696 0.0523 -0.9962>}
 dyn_hair_pigtail_06 {parent=dyn_hair_pigtail_05, loc=<7.4523 -0.0228 0.0127>, rot=<-0.0001 0.0048 -0.0162 -0.9999>}
 dyn_hair_back_01_base_l {parent=head, loc=<2.6413 -3.1916 -5.0198>, rot=<0.1904 -0.8813 0.3618 -0.2371>}
 dyn_hair_back_01_01_l {parent=dyn_hair_back_01_base_l, loc=<0 0 0>, rot=<0 0 0 1>}
 dyn_hair_back_01_02_l {parent=dyn_hair_back_01_01_l, loc=<7.1873 -0.0531 -0.198>, rot=<-0.1029 -0.064 -0.0028 -0.9926>}
 dyn_hair_back_01_03_l {parent=dyn_hair_back_01_02_l, loc=<7.8665 0.3412 -0.4341>, rot=<-0.0796 -0.0537 -0.0198 -0.9952>}
 dyn_hair_back_01_04_l {parent=dyn_hair_back_01_03_l, loc=<6.7974 0.0473 -0.4184>, rot=<-0.0131 -0.0495 -0.0279 -0.9983>}
 dyn_hair_back_01_05_l {parent=dyn_hair_back_01_04_l, loc=<8.0645 0.1072 -0.2822>, rot=<-0.0243 -0.0584 -0.008 -0.998>}
 dyn_hair_back_01_06_l {parent=dyn_hair_back_01_05_l, loc=<6.1968 0.2157 -0.4875>, rot=<0.023 -0.0608 -0.0526 -0.9965>}
 dyn_hair_back_02_base_l {parent=head, loc=<2.1354 -5.8397 -1.6714>, rot=<0.1809 -0.7421 0.6115 -0.2063>}
 dyn_hair_back_02_01_l {parent=dyn_hair_back_02_base_l, loc=<0 0 0>, rot=<0 0 0 1>}
 dyn_hair_back_02_02_l {parent=dyn_hair_back_02_01_l, loc=<6.3137 0.0399 0.0163>, rot=<0.0315 -0.0646 -0.0247 -0.9971>}
 dyn_hair_back_02_03_l {parent=dyn_hair_back_02_02_l, loc=<8.0844 -0.0825 -0.5942>, rot=<-0.0024 -0.0399 -0.0096 -0.9992>}
 dyn_hair_back_02_04_l {parent=dyn_hair_back_02_03_l, loc=<7.0661 -0.1236 -0.2876>, rot=<-0.007 -0.0571 -0.0092 -0.9983>}
 dyn_hair_back_02_05_l {parent=dyn_hair_back_02_04_l, loc=<7.8985 0.1019 -0.2099>, rot=<0.0193 -0.0583 0.0058 -0.9981>}
 dyn_hair_back_02_06_l {parent=dyn_hair_back_02_05_l, loc=<7.8705 0.1569 0.1269>, rot=<-0.013 -0.0411 -0.0172 -0.9989>}
 dyn_hair_back_01_base_r {parent=head, loc=<2.4245 -2.8517 5.5446>, rot=<0.1736 -0.3822 0.8896 -0.18>}
 dyn_hair_back_01_01_r {parent=dyn_hair_back_01_base_r, loc=<0 0 0>, rot=<0 0 0 1>}
 dyn_hair_back_01_02_r {parent=dyn_hair_back_01_01_r, loc=<7.7357 0.4552 0.5526>, rot=<0.1553 -0.0445 0.0188 -0.9867>}
 dyn_hair_back_01_03_r {parent=dyn_hair_back_01_02_r, loc=<7.7545 -0.2326 0.176>, rot=<0.0011 -0.0285 0.0379 -0.9989>}
 dyn_hair_back_01_04_r {parent=dyn_hair_back_01_03_r, loc=<7.1563 -0.3711 -0.2344>, rot=<-0.0023 -0.0571 -0.0065 -0.9983>}
 dyn_hair_back_01_05_r {parent=dyn_hair_back_01_04_r, loc=<7.8913 -0.4888 -0.5832>, rot=<0.0007 -0.0688 0.0308 -0.9972>}
 dyn_hair_back_01_06_r {parent=dyn_hair_back_01_05_r, loc=<6.3471 -0.0759 -0.0229>, rot=<-0.0003 -0.0299 0.0057 -0.9995>}
 dyn_hair_back_02_base_r {parent=head, loc=<1.8591 -5.3906 2.6466>, rot=<0.2101 -0.5518 0.7842 -0.1908>}
 dyn_hair_back_02_01_r {parent=dyn_hair_back_02_base_r, loc=<0 0 0>, rot=<0 0 0 1>}
 dyn_hair_back_02_02_r {parent=dyn_hair_back_02_01_r, loc=<6.5176 0.507 0.0052>, rot=<0.041 -0.0741 -0.0122 -0.9963>}
 dyn_hair_back_02_03_r {parent=dyn_hair_back_02_02_r, loc=<8.1384 -0.0234 -0.3915>, rot=<0.0008 -0.0794 0.0105 -0.9968>}
 dyn_hair_back_02_04_r {parent=dyn_hair_back_02_03_r, loc=<7.2836 -0.2345 0.5369>, rot=<0.0001 -0.0399 0.0344 -0.9986>}
 dyn_hair_back_02_05_r {parent=dyn_hair_back_02_04_r, loc=<7.7241 0.2837 -0.4904>, rot=<0.0007 -0.0387 -0.0183 -0.9991>}
 dyn_hair_back_02_06_r {parent=dyn_hair_back_02_05_r, loc=<7.969 0 -0.1455>, rot=<0 -0.0649 0 -0.9979>}
 deform_hair_back {parent=head, loc=<3.528 0.0826 -0.0063>, rot=<-0.0057 0.7052 0.709 -0.007>}
 clavicle_l {parent=spine_05, loc=<5.4831 -0.1053 -3.2946>, rot=<0.0465 -0.7289 -0.009 -0.6829>}
 upperarm_l {parent=clavicle_l, loc=<9.8658 -1.2136 -0.4105>, rot=<0.0164 -0.3245 -0.0269 -0.9454>}
 lowerarm_l {parent=upperarm_l, loc=<26.465 -0.8345 0.259>, rot=<0.0036 0.0161 -0.2298 -0.9731>}
 clavicle_r {parent=spine_05, loc=<5.4831 -0.1053 3.2946>, rot=<0.7289 0.0465 -0.6829 0.009>}
 upperarm_r {parent=clavicle_r, loc=<-9.8658 1.2136 0.4105>, rot=<0.0164 -0.3245 -0.0269 -0.9454>}
 lowerarm_r {parent=upperarm_r, loc=<-26.465 0.8345 -0.259>, rot=<0.0036 0.0161 -0.2298 -0.9731>}
 deform_hair_back_ALT_l {parent=spine_03, loc=<9.3192 -12.8715 -8.0364>, rot=<0.0305 -0.8233 0.5601 -0.0869>}
 deform_hair_back_ALT_r {parent=spine_03, loc=<9.319 -12.8716 8.0364>, rot=<0.8233 0.0305 -0.0869 -0.5601>}
]
variables [ItemToModify_0_Transform_Translation_X {type=double, default=0}]
