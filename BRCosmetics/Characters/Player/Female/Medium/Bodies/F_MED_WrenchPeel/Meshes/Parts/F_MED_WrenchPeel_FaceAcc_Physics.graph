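# Physics asset for the WrenchPeel female medium face accessory mesh, providing physics bodies for four dynamic back-hair chains.
# Each six-segment capsule chain hangs from a zero-size base locked in rotation to the head, with per-joint swing and twist limits.
# Cross constraints tie neighbouring chains together and link upper segments to spine_05.
# Pelvis, spine, neck, head, clavicle and arm capsules form the body, joined by mostly free angular constraints.
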PhysicsAsset /BRCosmetics/Characters/Player/Female/Medium/Bodies/F_MED_WrenchPeel/Meshes/Parts/F_MED_WrenchPeel_FaceAcc_Physics
bodies {
 clavicle_l [{
   shapeType=Capsule
   length=9.0643
   radius=6.4896
   shapeName=lowerarm_l_capsule
   center=<5.0132 -0.2334 0>
   rotation=<rot 84.7439 -0.8534 179.9983>
  }]
 clavicle_r [{
   shapeType=Capsule
   length=9.0643
   radius=6.4896
   shapeName=lowerarm_r_capsule
   center=<-5.0132 0.2334 0>
   rotation=<rot 84.7439 -0.8534 -0.0017>
  }]
 dyn_hair_back_01_01_l [{shapeType=Capsule, length=4, radius=5, shapeName=dyn_hair_back_01_06_r_capsule, center=<5.3304 0 3.9282>, rotation=<rot 90 0 0>}]
 dyn_hair_back_01_01_r [{shapeType=Capsule, length=4, radius=5, shapeName=dyn_hair_back_01_06_r_capsule, center=<5.3304 0 3.9282>, rotation=<rot 90 0 0>}]
 dyn_hair_back_01_02_l [{shapeType=Capsule, length=5.877, radius=5, shapeName=dyn_hair_back_01_06_r_capsule, center=<4 0 3.3695>, rotation=<rot 90 0 0>}]
 dyn_hair_back_01_02_r [{shapeType=Capsule, length=5.877, radius=5, shapeName=dyn_hair_back_01_06_r_capsule, center=<4 0 3.3695>, rotation=<rot 90 -4.534 0>}]
 dyn_hair_back_01_03_l [{shapeType=Capsule, length=5.877, radius=5, shapeName=dyn_hair_back_01_06_r_capsule, center=<4 0 2.5>, rotation=<rot 90 1.0216 0>}]
 dyn_hair_back_01_03_r [{shapeType=Capsule, length=5.877, radius=5, shapeName=dyn_hair_back_01_06_r_capsule, center=<4 0 2.5>, rotation=<rot 90 -1.8754 0>}]
 dyn_hair_back_01_04_l [{shapeType=Capsule, length=5.877, radius=5, shapeName=dyn_hair_back_01_06_r_capsule, center=<4 0 2.5>, rotation=<rot 90 1.5093 0>}]
 dyn_hair_back_01_04_r [{shapeType=Capsule, length=5.877, radius=5, shapeName=dyn_hair_back_01_06_r_capsule, center=<4 0 2.5>, rotation=<rot 90 -6.5994 0>}]
 dyn_hair_back_01_05_l [{
   shapeType=Capsule
   length=5.877
   radius=5
   shapeName=dyn_hair_back_01_06_r_capsule
   center=<4.0006 0.0814 2.2874>
   rotation=<rot 87.6326 2.6595 -0.0002>
  }]
 dyn_hair_back_01_05_r [{
   shapeType=Capsule
   length=5.877
   radius=5
   shapeName=dyn_hair_back_01_06_r_capsule
   center=<3.9745 -0.2291 2.3032>
   rotation=<rot 90 -6.3614 0>
  }]
 dyn_hair_back_01_06_l [{
   shapeType=Capsule
   length=2.8113
   radius=5.1887
   shapeName=dyn_hair_back_01_06_r_capsule
   center=<4.4387 0.7444 1.9307>
   rotation=<rot 90 4.7051 0>
  }]
 dyn_hair_back_01_06_r [{
   shapeType=Capsule
   length=2.8113
   radius=5.1887
   shapeName=dyn_hair_back_01_06_r_capsule
   center=<4.3173 -3.0743 2.0995>
   rotation=<rot 90 -8.5929 0>
  }]
 dyn_hair_back_01_base_l [{shapeType=Capsule, length=0, radius=0, shapeName=dyn_hair_back_01_base_l_capsule}]
 dyn_hair_back_01_base_r [{shapeType=Capsule, length=0, radius=0, shapeName=dyn_hair_back_01_base_r_capsule}]
 dyn_hair_back_02_01_l [{shapeType=Capsule, length=4, radius=5, shapeName=dyn_hair_back_01_06_r_capsule, center=<5.3304 0 3.9282>, rotation=<rot 90 0 0>}]
 dyn_hair_back_02_01_r [{shapeType=Capsule, length=4, radius=5, shapeName=dyn_hair_back_01_06_r_capsule, center=<5.3304 0 3.9282>, rotation=<rot 90 0 0>}]
 dyn_hair_back_02_02_l [{shapeType=Capsule, length=5.877, radius=5, shapeName=dyn_hair_back_01_06_r_capsule, center=<4 0 3.3695>, rotation=<rot 90 0 0>}]
 dyn_hair_back_02_02_r [{shapeType=Capsule, length=5.877, radius=5, shapeName=dyn_hair_back_01_06_r_capsule, center=<4 0 3.3695>, rotation=<rot 90 0 0>}]
 dyn_hair_back_02_03_l [{shapeType=Capsule, length=5.877, radius=5, shapeName=dyn_hair_back_01_06_r_capsule, center=<4 0 2.5>, rotation=<rot 90 0 0>}]
 dyn_hair_back_02_03_r [{shapeType=Capsule, length=5.877, radius=5, shapeName=dyn_hair_back_01_06_r_capsule, center=<4 0 2.5>, rotation=<rot 90 0 0>}]
 dyn_hair_back_02_04_l [{shapeType=Capsule, length=5.877, radius=5, shapeName=dyn_hair_back_01_06_r_capsule, center=<4 0 2.5>, rotation=<rot 90 0 0>}]
 dyn_hair_back_02_04_r [{shapeType=Capsule, length=5.877, radius=5, shapeName=dyn_hair_back_01_06_r_capsule, center=<4 0 2.5>, rotation=<rot 90 0 0>}]
 dyn_hair_back_02_05_l [{shapeType=Capsule, length=5.877, radius=5, shapeName=dyn_hair_back_01_06_r_capsule, center=<4 0 2.5>, rotation=<rot 90 0 0>}]
 dyn_hair_back_02_05_r [{shapeType=Capsule, length=5.877, radius=5, shapeName=dyn_hair_back_01_06_r_capsule, center=<4 0 2.5>, rotation=<rot 90 0 0>}]
 dyn_hair_back_02_06_l [{shapeType=Capsule, length=9.095, radius=5, shapeName=dyn_hair_back_01_06_r_capsule, center=<4.5 0 2.5>, rotation=<rot 90 90.5706 0>}]
 dyn_hair_back_02_06_r [{shapeType=Capsule, length=9.095, radius=5, shapeName=dyn_hair_back_01_06_r_capsule, center=<4.5 0 2.5>, rotation=<rot 90 90.5706 0>}]
 dyn_hair_back_02_base_l [{shapeType=Capsule, length=0, radius=0, shapeName=dyn_hair_back_02_base_l_capsule}]
 dyn_hair_back_02_base_r [{shapeType=Capsule, length=0, radius=0, shapeName=dyn_hair_back_02_base_r_capsule}]
 head [{shapeType=Capsule, length=0.3988, radius=8.0307, shapeName=head_capsule, center=<-2.2995 2.9425 0>, rotation=<rot 90 0 0>}]
 lowerarm_l [{shapeType=Capsule, length=12, radius=5, shapeName=lowerarm_l_capsule, center=<8 0 0>, rotation=<rot 90 180 0>}]
 lowerarm_r [{shapeType=Capsule, length=12, radius=5, shapeName=lowerarm_r_capsule, center=<-8 0 0>, rotation=<rot 90 0 0>}]
 neck_01 [{shapeType=Capsule, length=1.8452, radius=8.5747, shapeName=head_capsule, center=<1.3862 3.6116 0>, rotation=<rot 90 4.7448 0>}]
 pelvis [{shapeType=Capsule, length=12.959, radius=12.8919, shapeName=spine_05_capsule, center=<-0.3617 0.169 0>}]
 spine_01 [{shapeType=Capsule, length=9.9813, radius=9.2162, shapeName=spine_05_capsule, center=<3.0276 -1.9143 0>}]
 spine_02 [{shapeType=Capsule, length=9.6651, radius=8, shapeName=spine_05_capsule, center=<3.2057 -1.5843 0>}]
 spine_03 [{shapeType=Capsule, length=11.0989, radius=8, shapeName=spine_05_capsule, center=<3.1776 -1.1679 0>}]
 spine_04 [{shapeType=Capsule, length=13.509, radius=8, shapeName=spine_05_capsule, center=<3.3128 -0.5042 0>}]
 spine_05 [{shapeType=Capsule, length=18.3969, radius=8.1671, shapeName=spine_05_capsule, center=<3.7777 0.2427 0>}]
 upperarm_l [{
   shapeType=Capsule
   length=25.0027
   radius=4.8785
   shapeName=lowerarm_l_capsule
   center=<12.8541 -0.35 -0.9637>
   rotation=<rot 88.1565 179.1499 0.0016>
  }]
 upperarm_r [{
   shapeType=Capsule
   length=25.0027
   radius=4.8785
   shapeName=lowerarm_r_capsule
   center=<-12.8541 0.35 0.9637>
   rotation=<rot 88.1565 179.1499 -179.9984>
  }]
}
constraints [
 {bone1Name=dyn_hair_back_01_01_l, bone2Name=dyn_hair_back_01_base_l, swing1LimitDegrees=30, swing2LimitDegrees=30, twistLimitDegrees=5}
 {bone1Name=dyn_hair_back_01_02_l, bone2Name=dyn_hair_back_01_01_l, swing1LimitDegrees=25, swing2LimitDegrees=30, twistLimitDegrees=15}
 {bone1Name=dyn_hair_back_01_03_l, bone2Name=dyn_hair_back_01_02_l, swing1LimitDegrees=25, swing2LimitDegrees=25, twistLimitDegrees=15}
 {bone1Name=dyn_hair_back_01_04_l, bone2Name=dyn_hair_back_01_03_l, swing1LimitDegrees=35, swing2LimitDegrees=35, twistLimitDegrees=15}
 {bone1Name=dyn_hair_back_01_05_l, bone2Name=dyn_hair_back_01_04_l, swing1LimitDegrees=35, swing2LimitDegrees=35, twistLimitDegrees=15}
 {bone1Name=dyn_hair_back_01_06_l, bone2Name=dyn_hair_back_01_05_l, swing1LimitDegrees=35, swing2LimitDegrees=50, twistLimitDegrees=15}
 {bone1Name=dyn_hair_back_01_01_r, bone2Name=dyn_hair_back_01_base_r, swing1LimitDegrees=30, swing2LimitDegrees=30, twistLimitDegrees=5}
 {bone1Name=dyn_hair_back_01_02_r, bone2Name=dyn_hair_back_01_01_r, swing1LimitDegrees=25, swing2LimitDegrees=30, twistLimitDegrees=15}
 {bone1Name=dyn_hair_back_01_03_r, bone2Name=dyn_hair_back_01_02_r, swing1LimitDegrees=25, swing2LimitDegrees=25, twistLimitDegrees=15}
 {bone1Name=dyn_hair_back_01_04_r, bone2Name=dyn_hair_back_01_03_r, swing1LimitDegrees=35, swing2LimitDegrees=35, twistLimitDegrees=15}
 {bone1Name=dyn_hair_back_01_05_r, bone2Name=dyn_hair_back_01_04_r, swing1LimitDegrees=35, swing2LimitDegrees=35, twistLimitDegrees=15}
 {bone1Name=dyn_hair_back_01_06_r, bone2Name=dyn_hair_back_01_05_r, swing1LimitDegrees=35, swing2LimitDegrees=50, twistLimitDegrees=15}
 {bone1Name=dyn_hair_back_02_01_l, bone2Name=dyn_hair_back_02_base_l, swing1LimitDegrees=25, swing2LimitDegrees=15, twistLimitDegrees=5}
 {bone1Name=dyn_hair_back_02_02_l, bone2Name=dyn_hair_back_02_01_l, swing1LimitDegrees=25, swing2LimitDegrees=25, twistLimitDegrees=15}
 {bone1Name=dyn_hair_back_02_03_l, bone2Name=dyn_hair_back_02_02_l, swing1LimitDegrees=15, swing2LimitDegrees=25, twistLimitDegrees=15}
 {bone1Name=dyn_hair_back_02_04_l, bone2Name=dyn_hair_back_02_03_l, swing1LimitDegrees=25, swing2LimitDegrees=35, twistLimitDegrees=15}
 {bone1Name=dyn_hair_back_02_05_l, bone2Name=dyn_hair_back_02_04_l, swing1LimitDegrees=35, swing2LimitDegrees=35, twistLimitDegrees=15}
 {bone1Name=dyn_hair_back_02_06_l, bone2Name=dyn_hair_back_02_05_l, swing1LimitDegrees=35, swing2LimitDegrees=50, twistLimitDegrees=15}
 {bone1Name=dyn_hair_back_02_01_r, bone2Name=dyn_hair_back_02_base_r, swing1LimitDegrees=25, swing2LimitDegrees=15, twistLimitDegrees=5}
 {bone1Name=dyn_hair_back_02_03_r, bone2Name=dyn_hair_back_02_02_r, swing1LimitDegrees=15, swing2LimitDegrees=25, twistLimitDegrees=15}
 {bone1Name=dyn_hair_back_02_04_r, bone2Name=dyn_hair_back_02_03_r, swing1LimitDegrees=25, swing2LimitDegrees=35, twistLimitDegrees=15}
 {bone1Name=dyn_hair_back_02_05_r, bone2Name=dyn_hair_back_02_04_r, swing1LimitDegrees=35, swing2LimitDegrees=35, twistLimitDegrees=15}
 {bone1Name=dyn_hair_back_02_06_r, bone2Name=dyn_hair_back_02_05_r, swing1LimitDegrees=35, swing2LimitDegrees=50, twistLimitDegrees=15}
 {
  bone1Name=dyn_hair_back_01_base_l
  bone2Name=head
  swing1LimitDegrees=45
  swing1Motion=Locked
  swing2LimitDegrees=45
  swing2Motion=Locked
  twistLimitDegrees=45
  twistMotion=Locked
 }
 {
  bone1Name=dyn_hair_back_02_base_l
  bone2Name=head
  swing1LimitDegrees=45
  swing1Motion=Locked
  swing2LimitDegrees=45
  swing2Motion=Locked
  twistLimitDegrees=45
  twistMotion=Locked
 }
 {
  bone1Name=dyn_hair_back_01_base_r
  bone2Name=head
  swing1LimitDegrees=45
  swing1Motion=Locked
  swing2LimitDegrees=45
  swing2Motion=Locked
  twistLimitDegrees=45
  twistMotion=Locked
 }
 {
  bone1Name=dyn_hair_back_02_base_r
  bone2Name=head
  swing1LimitDegrees=45
  swing1Motion=Locked
  swing2LimitDegrees=45
  swing2Motion=Locked
  twistLimitDegrees=45
  twistMotion=Locked
 }
 {
  bone1Name=head
  bone2Name=neck_01
  swing1LimitDegrees=45
  swing1Motion=Free
  swing2LimitDegrees=45
  swing2Motion=Free
  twistLimitDegrees=45
  twistMotion=Locked
 }
 {
  bone1Name=spine_02
  bone2Name=spine_01
  swing1LimitDegrees=45
  swing1Motion=Free
  swing2LimitDegrees=45
  swing2Motion=Free
  twistLimitDegrees=45
  twistMotion=Free
 }
 {
  bone1Name=spine_03
  bone2Name=spine_02
  swing1LimitDegrees=45
  swing1Motion=Free
  swing2LimitDegrees=45
  swing2Motion=Free
  twistLimitDegrees=45
  twistMotion=Free
 }
 {
  bone1Name=spine_04
  bone2Name=spine_03
  swing1LimitDegrees=45
  swing1Motion=Free
  swing2LimitDegrees=45
  swing2Motion=Free
  twistLimitDegrees=45
  twistMotion=Free
 }
 {
  bone1Name=spine_05
  bone2Name=spine_04
  swing1LimitDegrees=45
  swing1Motion=Free
  swing2LimitDegrees=45
  swing2Motion=Free
  twistLimitDegrees=45
  twistMotion=Free
 }
 {bone1Name=dyn_hair_back_02_05_r, bone2Name=dyn_hair_back_01_05_r, swing1LimitDegrees=35, swing2LimitDegrees=35, twistLimitDegrees=35}
 {bone1Name=dyn_hair_back_02_05_l, bone2Name=dyn_hair_back_01_05_l, swing1LimitDegrees=35, swing2LimitDegrees=35, twistLimitDegrees=35}
 {bone1Name=dyn_hair_back_02_05_l, bone2Name=dyn_hair_back_02_05_r, swing1LimitDegrees=35, swing2LimitDegrees=35, twistLimitDegrees=35}
 {bone1Name=dyn_hair_back_02_06_r, bone2Name=dyn_hair_back_01_06_r, swing1LimitDegrees=35, swing2LimitDegrees=35, twistLimitDegrees=35}
 {bone1Name=dyn_hair_back_02_06_l, bone2Name=dyn_hair_back_01_06_l, swing1LimitDegrees=35, swing2LimitDegrees=35, twistLimitDegrees=35}
 {
  bone1Name=clavicle_l
  bone2Name=spine_05
  swing1LimitDegrees=45
  swing1Motion=Free
  swing2LimitDegrees=45
  swing2Motion=Free
  twistLimitDegrees=45
  twistMotion=Free
 }
 {
  bone1Name=upperarm_l
  bone2Name=clavicle_l
  swing1LimitDegrees=45
  swing1Motion=Free
  swing2LimitDegrees=45
  swing2Motion=Free
  twistLimitDegrees=45
  twistMotion=Free
 }
 {
  bone1Name=lowerarm_l
  bone2Name=upperarm_l
  swing1LimitDegrees=45
  swing1Motion=Free
  swing2LimitDegrees=45
  swing2Motion=Free
  twistLimitDegrees=45
  twistMotion=Free
 }
 {
  bone1Name=clavicle_r
  bone2Name=spine_05
  swing1LimitDegrees=45
  swing1Motion=Free
  swing2LimitDegrees=45
  swing2Motion=Free
  twistLimitDegrees=45
  twistMotion=Free
 }
 {
  bone1Name=upperarm_r
  bone2Name=clavicle_r
  swing1LimitDegrees=45
  swing1Motion=Free
  swing2LimitDegrees=45
  swing2Motion=Free
  twistLimitDegrees=45
  twistMotion=Free
 }
 {
  bone1Name=lowerarm_r
  bone2Name=upperarm_r
  swing1LimitDegrees=45
  swing1Motion=Free
  swing2LimitDegrees=45
  swing2Motion=Free
  twistLimitDegrees=45
  twistMotion=Free
 }
 {bone1Name=dyn_hair_back_02_06_l, bone2Name=dyn_hair_back_02_06_r, swing1LimitDegrees=35, swing2LimitDegrees=35, twistLimitDegrees=35}
 {bone1Name=dyn_hair_back_02_04_l, bone2Name=dyn_hair_back_02_04_r, swing1LimitDegrees=35, swing2LimitDegrees=35, twistLimitDegrees=35}
 {bone1Name=dyn_hair_back_02_03_r, bone2Name=dyn_hair_back_01_03_r, swing1LimitDegrees=35, swing2LimitDegrees=35, twistLimitDegrees=35}
 {bone1Name=dyn_hair_back_02_03_l, bone2Name=dyn_hair_back_01_03_l, swing1LimitDegrees=35, swing2LimitDegrees=35, twistLimitDegrees=35}
 {bone1Name=spine_05, bone2Name=dyn_hair_back_02_03_r, swing1LimitDegrees=45, swing2LimitDegrees=45, twistLimitDegrees=45}
 {bone1Name=spine_05, bone2Name=dyn_hair_back_02_03_l, swing1LimitDegrees=45, swing2LimitDegrees=45, twistLimitDegrees=45}
 {
  bone1Name=spine_01
  bone2Name=pelvis
  swing1LimitDegrees=45
  swing1Motion=Free
  swing2LimitDegrees=45
  swing2Motion=Free
  twistLimitDegrees=45
  twistMotion=Free
 }
 {bone1Name=spine_05, bone2Name=dyn_hair_back_01_03_r, swing1LimitDegrees=45, swing2LimitDegrees=45, twistLimitDegrees=45}
 {bone1Name=spine_05, bone2Name=dyn_hair_back_01_02_r, swing1LimitDegrees=45, swing2LimitDegrees=45, twistLimitDegrees=45}
 {bone1Name=spine_05, bone2Name=dyn_hair_back_01_03_l, swing1LimitDegrees=45, swing2LimitDegrees=45, twistLimitDegrees=45}
 {
  bone1Name=neck_01
  bone2Name=spine_05
  swing1LimitDegrees=45
  swing1Motion=Free
  swing2LimitDegrees=45
  swing2Motion=Free
  twistLimitDegrees=45
  twistMotion=Free
 }
 {bone1Name=spine_05, bone2Name=dyn_hair_back_01_02_l, swing1LimitDegrees=45, swing2LimitDegrees=45, twistLimitDegrees=45}
 {bone1Name=dyn_hair_back_02_02_r, bone2Name=dyn_hair_back_02_01_r, swing1LimitDegrees=25, swing2LimitDegrees=25, twistLimitDegrees=15}
 {bone1Name=spine_05, bone2Name=dyn_hair_back_02_02_l, swing1LimitDegrees=45, swing2LimitDegrees=45, twistLimitDegrees=45}
 {bone1Name=spine_05, bone2Name=dyn_hair_back_02_02_r, swing1LimitDegrees=45, swing2LimitDegrees=45, twistLimitDegrees=45}
 {bone1Name=dyn_hair_back_02_04_r, bone2Name=dyn_hair_back_01_04_r, swing1LimitDegrees=35, swing2LimitDegrees=35, twistLimitDegrees=35}
 {bone1Name=dyn_hair_back_02_04_l, bone2Name=dyn_hair_back_01_04_l, swing1LimitDegrees=35, swing2LimitDegrees=35, twistLimitDegrees=35}
]
