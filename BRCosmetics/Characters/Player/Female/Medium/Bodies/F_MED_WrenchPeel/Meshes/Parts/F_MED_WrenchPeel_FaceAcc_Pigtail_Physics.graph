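# Physics asset for the pigtail hair of the F_MED_WrenchPeel face accessory, defining capsule bodies and joints for its motion.
# Capsules cover the pigtail chain, two left/right sets of back-hair bones, and head, neck, spine and upper arms.
# Angle-limited constraints chain the pigtail from a zero-size base.
# Extra limited and free-rotation constraints cross-link it to the right back_02 bones only; other back-hair and body capsules have no constraints.
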PhysicsAsset /BRCosmetics/Characters/Player/Female/Medium/Bodies/F_MED_WrenchPeel/Meshes/Parts/F_MED_WrenchPeel_FaceAcc_Pigtail_Physics
bodies {
 dyn_hair_back_01_01_l [{
   shapeType=Capsule
   length=1.1214
   radius=5.7444
   shapeName=dyn_hair_back_01_06_r_capsule
   center=<5.3304 0 -4.2898>
   rotation=<rot 82.7293 -2.1991 -0.0003>
  }]
 dyn_hair_back_01_01_r [{
   shapeType=Capsule
   length=1.1214
   radius=5.7444
   shapeName=dyn_hair_back_01_06_r_capsule
   center=<5.3304 0 -4.2898>
   rotation=<rot 82.7293 0.0001 0.0002>
  }]
 dyn_hair_back_01_02_l [{
   shapeType=Capsule
   length=3.7412
   radius=5.0478
   shapeName=dyn_hair_back_01_06_r_capsule
   center=<4.1983 1.8935 -3.405>
   rotation=<rot 82.9408 -4.5339 -10.7191>
  }]
 dyn_hair_back_01_02_r [{
   shapeType=Capsule
   length=3.7412
   radius=5.0478
   shapeName=dyn_hair_back_01_06_r_capsule
   center=<3.7411 -2.7696 -4.0263>
   rotation=<rot 87.3788 -4.5337 0.0003>
  }]
 dyn_hair_back_01_03_l [{
   shapeType=Capsule
   length=1.1214
   radius=5.7444
   shapeName=dyn_hair_back_01_06_r_capsule
   center=<3.9426 -0.1958 -4.3086>
   rotation=<rot 87.0745 -1.8755 -5.1397>
  }]
 dyn_hair_back_01_03_r [{
   shapeType=Capsule
   length=1.1214
   radius=5.7444
   shapeName=dyn_hair_back_01_06_r_capsule
   center=<3.9491 0.0017 -4.3086>
   rotation=<rot 87.0746 -1.8753 0.0002>
  }]
 dyn_hair_back_01_04_l [{
   shapeType=Capsule
   length=1.1214
   radius=5.7444
   shapeName=dyn_hair_back_01_06_r_capsule
   center=<3.9572 -0.1886 -3.9168>
   rotation=<rot 90 -0.0011 0>
  }]
 dyn_hair_back_01_04_r [{
   shapeType=Capsule
   length=1.1214
   radius=5.7444
   shapeName=dyn_hair_back_01_06_r_capsule
   center=<3.9572 -0.3696 -3.9168>
   rotation=<rot 90 -6.5994 0>
  }]
 dyn_hair_back_01_05_l [{
   shapeType=Capsule
   length=1.1214
   radius=5.7444
   shapeName=dyn_hair_back_01_06_r_capsule
   center=<3.7639 -0.1409 -3.9123>
   rotation=<rot 84.1446 3.9862 -0.0003>
  }]
 dyn_hair_back_01_05_r [{
   shapeType=Capsule
   length=1.1214
   radius=5.7444
   shapeName=dyn_hair_back_01_06_r_capsule
   center=<3.9391 -0.8862 -3.9168>
   rotation=<rot 90 -3.9338 0>
  }]
 dyn_hair_back_01_06_l [{
   shapeType=Capsule
   length=4.4212
   radius=6.5036
   shapeName=dyn_hair_back_01_06_r_capsule
   center=<6.3606 0.4018 -4.6626>
   rotation=<rot 90 0.9932 0>
  }]
 dyn_hair_back_01_06_r [{
   shapeType=Capsule
   length=4.4212
   radius=6.5036
   shapeName=dyn_hair_back_01_06_r_capsule
   center=<5.9614 -2.3819 -4.6626>
   rotation=<rot 90 -8.5929 0>
  }]
 dyn_hair_back_02_01_l [{shapeType=Capsule, length=1.1214, radius=5.7444, shapeName=dyn_hair_back_01_06_r_capsule, center=<4.4562 0 -4.8329>, rotation=<rot 90 0 0>}]
 dyn_hair_back_02_01_r [{shapeType=Capsule, length=1.1214, radius=5.7444, shapeName=dyn_hair_back_01_06_r_capsule, center=<4.4562 0 -4.8329>, rotation=<rot 90 0 0>}]
 dyn_hair_back_02_02_l [{shapeType=Capsule, length=1.1214, radius=5.7444, shapeName=dyn_hair_back_01_06_r_capsule, center=<4.4562 0 -4.4341>, rotation=<rot 90 0 0>}]
 dyn_hair_back_02_02_r [{shapeType=Capsule, length=1.1214, radius=5.7444, shapeName=dyn_hair_back_01_06_r_capsule, center=<4.4562 0 -4.4341>, rotation=<rot 90 0 0>}]
 dyn_hair_back_02_03_l [{shapeType=Capsule, length=3.2304, radius=5.7444, shapeName=dyn_hair_back_01_06_r_capsule, center=<4.4562 0 -4.153>, rotation=<rot 90 0 0>}]
 dyn_hair_back_02_03_r [{shapeType=Capsule, length=3.2304, radius=5.7444, shapeName=dyn_hair_back_01_06_r_capsule, center=<4.4562 0 -4.153>, rotation=<rot 90 0 0>}]
 dyn_hair_back_02_04_l [{shapeType=Capsule, length=1.1214, radius=5.7444, shapeName=dyn_hair_back_01_06_r_capsule, center=<4.4562 0 -3.9141>, rotation=<rot 90 0 0>}]
 dyn_hair_back_02_04_r [{shapeType=Capsule, length=1.1214, radius=5.7444, shapeName=dyn_hair_back_01_06_r_capsule, center=<4.4562 0 -3.9141>, rotation=<rot 90 0 0>}]
 dyn_hair_back_02_05_l [{shapeType=Capsule, length=1.1214, radius=5.7444, shapeName=dyn_hair_back_01_06_r_capsule, center=<4.4562 0 -4.153>, rotation=<rot 90 0 0>}]
 dyn_hair_back_02_05_r [{shapeType=Capsule, length=1.1214, radius=5.7444, shapeName=dyn_hair_back_01_06_r_capsule, center=<4.4562 0 -4.153>, rotation=<rot 90 0 0>}]
 dyn_hair_back_02_06_l [{shapeType=Capsule, length=2.8227, radius=6.1131, shapeName=dyn_hair_back_01_06_r_capsule, center=<5.1865 0 -4.7606>, rotation=<rot 90 0 0>}]
 dyn_hair_back_02_06_r [{shapeType=Capsule, length=2.8227, radius=6.1131, shapeName=dyn_hair_back_01_06_r_capsule, center=<5.1865 0 -4.7606>, rotation=<rot 90 0 0>}]
 dyn_hair_pigtail_01 [{shapeType=Capsule, length=8.5653, radius=3.5875, shapeName=dyn_hair_pigtail_05_capsule, center=<2.95 0 3>, rotation=<rot 76.189 0.001 0.001>}]
 dyn_hair_pigtail_02 [{shapeType=Capsule, length=9.7022, radius=3.5875, shapeName=dyn_hair_pigtail_05_capsule, center=<4.7 0 3>, rotation=<rot 82.8081 0 0>}]
 dyn_hair_pigtail_03 [{
   shapeType=Capsule
   length=10.1754
   radius=3.1143
   shapeName=dyn_hair_pigtail_05_capsule
   center=<4.94 0 2.7>
   rotation=<rot 88.4005 0.0001 0.0001>
  }]
 dyn_hair_pigtail_04 [{shapeType=Capsule, length=9.9582, radius=3.3315, shapeName=dyn_hair_pigtail_05_capsule, center=<5 0 2.57>, rotation=<rot 90 0 0>}]
 dyn_hair_pigtail_05 [{shapeType=Capsule, length=9.9582, radius=3.3315, shapeName=dyn_hair_pigtail_05_capsule, center=<5 0 2.57>, rotation=<rot 90 0 0>}]
 dyn_hair_pigtail_06 [{shapeType=Capsule, length=7.9589, radius=3.3315, shapeName=dyn_hair_pigtail_06_capsule, center=<6.45 0 1.87>, rotation=<rot 88.7439 0 0>}]
 dyn_hair_pigtail_base [{shapeType=Capsule, length=0, radius=0, shapeName=dyn_hair_pigtail_base_capsule}]
 head [{shapeType=Capsule, length=4.4895, radius=8.0307, shapeName=head_capsule, center=<-2.2995 2.2942 0>, rotation=<rot 90 0 0>}]
 neck_01 [{shapeType=Capsule, length=1.8452, radius=8.5747, shapeName=head_capsule, center=<1.3862 3.6116 0>, rotation=<rot 90 4.7448 0>}]
 spine_02 [{shapeType=Capsule, length=12.3207, radius=8, shapeName=spine_05_capsule, center=<3.2057 0 0>}]
 spine_03 [{shapeType=Capsule, length=15.4739, radius=8, shapeName=spine_05_capsule, center=<3.1776 0 0>}]
 spine_04 [{shapeType=Capsule, length=18.6288, radius=8, shapeName=spine_05_capsule, center=<3.3128 0 0>}]
 spine_05 [{shapeType=Capsule, length=18.82, radius=9.4885, shapeName=spine_05_capsule, center=<3.3128 1.3672 0>}]
 upperarm_l [{shapeType=Capsule, length=24.8812, radius=5, shapeName=lowerarm_l_capsule, center=<12.8481 -0.3499 0>, rotation=<rot 90 179.1482 0>}]
 upperarm_r [{shapeType=Capsule, length=24.8812, radius=5, shapeName=lowerarm_r_capsule, center=<-12.8481 0.3499 0>, rotation=<rot 90 -0.8518 0>}]
}
constraints [
 {bone1Name=dyn_hair_pigtail_06, bone2Name=dyn_hair_pigtail_05, swing1LimitDegrees=45, swing2LimitDegrees=45, twistLimitDegrees=20}
 {bone1Name=dyn_hair_pigtail_01, bone2Name=dyn_hair_pigtail_base, swing1LimitDegrees=45, swing2LimitDegrees=45, twistLimitDegrees=25}
 {bone1Name=dyn_hair_pigtail_02, bone2Name=dyn_hair_pigtail_01, swing1LimitDegrees=45, swing2LimitDegrees=45, twistLimitDegrees=20}
 {bone1Name=dyn_hair_pigtail_03, bone2Name=dyn_hair_pigtail_02, swing1LimitDegrees=45, swing2LimitDegrees=45, twistLimitDegrees=20}
 {bone1Name=dyn_hair_pigtail_04, bone2Name=dyn_hair_pigtail_03, swing1LimitDegrees=45, swing2LimitDegrees=45, twistLimitDegrees=20}
 {bone1Name=dyn_hair_pigtail_05, bone2Name=dyn_hair_pigtail_04, swing1LimitDegrees=45, swing2LimitDegrees=45, twistLimitDegrees=20}
 {bone1Name=dyn_hair_back_02_03_r, bone2Name=dyn_hair_pigtail_03, swing1LimitDegrees=45, swing2LimitDegrees=45, twistLimitDegrees=45}
 {bone1Name=dyn_hair_back_02_02_r, bone2Name=dyn_hair_pigtail_02, swing1LimitDegrees=45, swing2LimitDegrees=45, twistLimitDegrees=45}
 {bone1Name=dyn_hair_back_02_01_r, bone2Name=dyn_hair_pigtail_01, swing1LimitDegrees=45, swing2LimitDegrees=45, twistLimitDegrees=5}
 {
  bone1Name=dyn_hair_pigtail_05
  bone2Name=dyn_hair_back_02_05_r
  swing1LimitDegrees=45
  swing1Motion=Free
  swing2LimitDegrees=45
  swing2Motion=Free
  twistLimitDegrees=45
  twistMotion=Free
 }
 {
  bone1Name=dyn_hair_pigtail_04
  bone2Name=dyn_hair_back_02_04_r
  swing1LimitDegrees=45
  swing1Motion=Free
  swing2LimitDegrees=45
  swing2Motion=Free
  twistLimitDegrees=45
  twistMotion=Free
 }
 {
  bone1Name=dyn_hair_pigtail_03
  bone2Name=dyn_hair_back_02_03_r
  swing1LimitDegrees=45
  swing1Motion=Free
  swing2LimitDegrees=45
  swing2Motion=Free
  twistLimitDegrees=45
  twistMotion=Free
 }
 {
  bone1Name=dyn_hair_pigtail_02
  bone2Name=dyn_hair_back_02_02_r
  swing1LimitDegrees=45
  swing1Motion=Free
  swing2LimitDegrees=45
  swing2Motion=Free
  twistLimitDegrees=45
  twistMotion=Free
 }
 {
  bone1Name=dyn_hair_pigtail_06
  bone2Name=dyn_hair_back_02_06_r
  swing1LimitDegrees=45
  swing1Motion=Free
  swing2LimitDegrees=45
  swing2Motion=Free
  twistLimitDegrees=45
  twistMotion=Free
 }
]
