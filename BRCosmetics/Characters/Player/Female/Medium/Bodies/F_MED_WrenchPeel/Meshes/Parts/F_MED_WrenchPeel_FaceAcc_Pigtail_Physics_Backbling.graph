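# Physics asset for the WrenchPeel female character's pigtail face accessory, Backbling variant, simulating four six-bone back-hair chains and a pigtail chain from the head.
# Limited swing and twist constraints run down each chain, with cross-links tying neighboring strands, the pigtail to the right back hair, and upper strands to spine_05.
# Capsules on head, neck, spine, pelvis and upper arms form the body, including a large offset spine_02 capsule, apparently back bling clearance.
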PhysicsAsset /BRCosmetics/Characters/Player/Female/Medium/Bodies/F_MED_WrenchPeel/Meshes/Parts/F_MED_WrenchPeel_FaceAcc_Pigtail_Physics_Backbling
bodies {
 dyn_hair_back_01_01_l [{
   shapeType=Capsule
   length=1.1214
   radius=5.7444
   shapeName=dyn_hair_back_01_06_r_capsule
   center=<5.2816 0.0019 -4.673>
   rotation=<rot 82.7293 -2.1991 -0.0003>
  }]
 dyn_hair_back_01_01_r [{
   shapeType=Capsule
   length=1.1214
   radius=5.7444
   shapeName=dyn_hair_back_01_06_r_capsule
   center=<5.2816 0 -4.673>
   rotation=<rot 82.7293 0.0001 0.0002>
  }]
 dyn_hair_back_01_02_l [{
   shapeType=Capsule
   length=3.7412
   radius=5.0478
   shapeName=dyn_hair_back_01_06_r_capsule
   center=<4.151 1.8973 -3.7884>
   rotation=<rot 82.9408 -4.5339 -10.7191>
  }]
 dyn_hair_back_01_02_r [{
   shapeType=Capsule
   length=3.7412
   radius=5.0478
   shapeName=dyn_hair_back_01_06_r_capsule
   center=<3.7235 -2.7682 -4.4122>
   rotation=<rot 87.3788 -4.5337 0.0003>
  }]
 dyn_hair_back_01_03_l [{
   shapeType=Capsule
   length=1.1214
   radius=5.7444
   shapeName=dyn_hair_back_01_06_r_capsule
   center=<3.9229 -0.1952 -4.6943>
   rotation=<rot 87.0745 -1.8755 -5.1397>
  }]
 dyn_hair_back_01_03_r [{
   shapeType=Capsule
   length=1.1214
   radius=5.7444
   shapeName=dyn_hair_back_01_06_r_capsule
   center=<3.9294 0.0023 -4.6943>
   rotation=<rot 87.0746 -1.8753 0.0002>
  }]
 dyn_hair_back_01_04_l [{
   shapeType=Capsule
   length=1.1214
   radius=5.7444
   shapeName=dyn_hair_back_01_06_r_capsule
   center=<3.9572 -0.1886 -4.5998>
   rotation=<rot 90 -0.0011 0>
  }]
 dyn_hair_back_01_04_r [{
   shapeType=Capsule
   length=1.1214
   radius=5.7444
   shapeName=dyn_hair_back_01_06_r_capsule
   center=<3.9572 -0.3696 -4.5998>
   rotation=<rot 90 -6.5994 0>
  }]
 dyn_hair_back_01_05_l [{
   shapeType=Capsule
   length=1.1214
   radius=5.7444
   shapeName=dyn_hair_back_01_06_r_capsule
   center=<3.6777 -0.147 -4.755>
   rotation=<rot 84.1446 3.9862 -0.0003>
  }]
 dyn_hair_back_01_05_r [{
   shapeType=Capsule
   length=1.1214
   radius=5.7444
   shapeName=dyn_hair_back_01_06_r_capsule
   center=<3.9391 -0.8862 -4.7638>
   rotation=<rot 90 -3.9338 0>
  }]
 dyn_hair_back_01_06_l [{
   shapeType=Capsule
   length=4.4212
   radius=6.5036
   shapeName=dyn_hair_back_01_06_r_capsule
   center=<6.3606 0.4018 -5.8812>
   rotation=<rot 90 0.9932 0>
  }]
 dyn_hair_back_01_06_r [{
   shapeType=Capsule
   length=4.4212
   radius=6.5036
   shapeName=dyn_hair_back_01_06_r_capsule
   center=<5.9614 -2.3819 -5.8812>
   rotation=<rot 90 -8.5929 0>
  }]
 dyn_hair_back_02_01_l [{shapeType=Capsule, length=1.1214, radius=5.7444, shapeName=dyn_hair_back_01_06_r_capsule, center=<4.4562 0 -5.2192>, rotation=<rot 90 0 0>}]
 dyn_hair_back_02_01_r [{shapeType=Capsule, length=1.1214, radius=5.7444, shapeName=dyn_hair_back_01_06_r_capsule, center=<4.4562 0 -5.2192>, rotation=<rot 90 0 0>}]
 dyn_hair_back_02_02_l [{shapeType=Capsule, length=1.1214, radius=5.7444, shapeName=dyn_hair_back_01_06_r_capsule, center=<4.4562 0 -4.8204>, rotation=<rot 90 0 0>}]
 dyn_hair_back_02_02_r [{shapeType=Capsule, length=1.1214, radius=5.7444, shapeName=dyn_hair_back_01_06_r_capsule, center=<4.4562 0 -4.8204>, rotation=<rot 90 0 0>}]
 dyn_hair_back_02_03_l [{shapeType=Capsule, length=1.1214, radius=5.7444, shapeName=dyn_hair_back_01_06_r_capsule, center=<4.4562 0 -4.5393>, rotation=<rot 90 0 0>}]
 dyn_hair_back_02_03_r [{shapeType=Capsule, length=1.1214, radius=5.7444, shapeName=dyn_hair_back_01_06_r_capsule, center=<4.4562 0 -4.5393>, rotation=<rot 90 0 0>}]
 dyn_hair_back_02_04_l [{shapeType=Capsule, length=1.1214, radius=5.7444, shapeName=dyn_hair_back_01_06_r_capsule, center=<4.4562 0 -4.5971>, rotation=<rot 90 0 0>}]
 dyn_hair_back_02_04_r [{shapeType=Capsule, length=1.1214, radius=5.7444, shapeName=dyn_hair_back_01_06_r_capsule, center=<4.4562 0 -4.5971>, rotation=<rot 90 0 0>}]
 dyn_hair_back_02_05_l [{
   shapeType=Capsule
   length=6.8765
   radius=5.7444
   shapeName=dyn_hair_back_01_06_r_capsule
   center=<4.4562 0 -5.0001>
   rotation=<rot 83.4523 -89.5872 -179.9996>
  }]
 dyn_hair_back_02_05_r [{
   shapeType=Capsule
   length=7.17
   radius=5.7444
   shapeName=dyn_hair_back_01_06_r_capsule
   center=<4.4562 0 -5.0001>
   rotation=<rot 89.8356 -89.4951 -179.9935>
  }]
 dyn_hair_back_02_06_l [{
   shapeType=Capsule
   length=9.494
   radius=6.1131
   shapeName=dyn_hair_back_01_06_r_capsule
   center=<5.1865 0 -5.9792>
   rotation=<rot 80.7474 -89.3686 -179.9992>
  }]
 dyn_hair_back_02_06_r [{
   shapeType=Capsule
   length=9.494
   radius=6.1131
   shapeName=dyn_hair_back_01_06_r_capsule
   center=<5.1865 0 -5.9792>
   rotation=<rot 88.5704 -89.3645 -179.9949>
  }]
 dyn_hair_pigtail_01 [{
   shapeType=Capsule
   length=8.5653
   radius=3.5875
   shapeName=dyn_hair_pigtail_05_capsule
   center=<2.9461 0 3.081>
   rotation=<rot 76.189 0.001 0.001>
  }]
 dyn_hair_pigtail_02 [{
   shapeType=Capsule
   length=10.5063
   radius=1.6927
   shapeName=dyn_hair_pigtail_05_capsule
   center=<4.3488 0 0.325>
   rotation=<rot 82.8081 0.0019 0.0019>
  }]
 dyn_hair_pigtail_03 [{
   shapeType=Capsule
   length=10.5693
   radius=1.5424
   shapeName=dyn_hair_pigtail_05_capsule
   center=<4.2053 0 0.9479>
   rotation=<rot 88.4005 0.0001 0.0001>
  }]
 dyn_hair_pigtail_04 [{
   shapeType=Capsule
   length=9.7416
   radius=1.4786
   shapeName=dyn_hair_pigtail_05_capsule
   center=<4.2343 0 0.7942>
   rotation=<rot 86.6281 -0.0001 -0.0001>
  }]
 dyn_hair_pigtail_05 [{shapeType=Capsule, length=7.9636, radius=1.6748, shapeName=dyn_hair_pigtail_05_capsule, center=<3.8168 0 0.9325>, rotation=<rot 90 0 0>}]
 dyn_hair_pigtail_06 [{shapeType=Capsule, length=6.3314, radius=1.667, shapeName=dyn_hair_pigtail_06_capsule, center=<4.2795 0 0.748>, rotation=<rot 88.7439 0 0>}]
 dyn_hair_pigtail_base [{shapeType=Capsule, length=0, radius=0, shapeName=dyn_hair_pigtail_base_capsule}]
 head [{shapeType=Capsule, length=4.4895, radius=8.0307, shapeName=head_capsule, center=<-2.2995 2.2942 0>, rotation=<rot 90 0 0>}]
 neck_01 [{shapeType=Capsule, length=1.8452, radius=8.5747, shapeName=head_capsule, center=<1.3862 3.6116 0>, rotation=<rot 90 4.7448 0>}]
 pelvis [{shapeType=Capsule, length=8.7237, radius=13.5676, shapeName=spine_05_capsule, center=<-1.9678 0.559 0>}]
 spine_02 [
  {shapeType=Capsule, length=12.3207, radius=8, shapeName=spine_05_capsule, center=<3.2057 0 0>}
  {shapeType=Capsule, length=28.8751, radius=41.0275, shapeName=spine_03_capsule, center=<3.7025 -62.1326 0>, rotation=<rot 0 4.5013 0>}
 ]
 spine_03 [{shapeType=Capsule, length=11.0989, radius=8, shapeName=spine_05_capsule, center=<3.1776 1.5145 0>}]
 spine_04 [{shapeType=Capsule, length=18.6288, radius=8, shapeName=spine_05_capsule, center=<3.3128 0 0>}]
 spine_05 [{shapeType=Capsule, length=18.82, radius=9.4885, shapeName=spine_05_capsule, center=<3.3128 1.3672 0>}]
 upperarm_l [{shapeType=Capsule, length=24.8812, radius=5, shapeName=lowerarm_l_capsule, center=<12.8481 -0.3499 0>, rotation=<rot 90 179.1482 0>}]
 upperarm_r [{shapeType=Capsule, length=24.8812, radius=5, shapeName=lowerarm_r_capsule, center=<-12.8481 0.3499 0>, rotation=<rot 90 -0.8518 0>}]
}
constraints [
 {bone1Name=dyn_hair_back_01_01_l, bone2Name=head, swing1LimitDegrees=30, swing2LimitDegrees=35, twistLimitDegrees=5}
 {bone1Name=dyn_hair_back_01_02_l, bone2Name=dyn_hair_back_01_01_l, swing1LimitDegrees=25, swing2LimitDegrees=30, twistLimitDegrees=15}
 {bone1Name=dyn_hair_back_01_03_l, bone2Name=dyn_hair_back_01_02_l, swing1LimitDegrees=25, swing2LimitDegrees=25, twistLimitDegrees=15}
 {bone1Name=dyn_hair_back_01_04_l, bone2Name=dyn_hair_back_01_03_l, swing1LimitDegrees=35, swing2LimitDegrees=35, twistLimitDegrees=15}
 {bone1Name=dyn_hair_back_01_05_l, bone2Name=dyn_hair_back_01_04_l, swing1LimitDegrees=35, swing2LimitDegrees=35, twistLimitDegrees=15}
 {bone1Name=dyn_hair_back_01_06_l, bone2Name=dyn_hair_back_01_05_l, swing1LimitDegrees=35, swing2LimitDegrees=50, twistLimitDegrees=15}
 {bone1Name=dyn_hair_back_01_01_r, bone2Name=head, swing1LimitDegrees=30, swing2LimitDegrees=35, twistLimitDegrees=5}
 {bone1Name=dyn_hair_back_01_02_r, bone2Name=dyn_hair_back_01_01_r, swing1LimitDegrees=25, swing2LimitDegrees=30, twistLimitDegrees=15}
 {bone1Name=dyn_hair_back_01_03_r, bone2Name=dyn_hair_back_01_02_r, swing1LimitDegrees=25, swing2LimitDegrees=25, twistLimitDegrees=15}
 {bone1Name=dyn_hair_back_01_04_r, bone2Name=dyn_hair_back_01_03_r, swing1LimitDegrees=35, swing2LimitDegrees=35, twistLimitDegrees=15}
 {bone1Name=dyn_hair_back_01_05_r, bone2Name=dyn_hair_back_01_04_r, swing1LimitDegrees=35, swing2LimitDegrees=35, twistLimitDegrees=15}
 {bone1Name=dyn_hair_back_01_06_r, bone2Name=dyn_hair_back_01_05_r, swing1LimitDegrees=35, swing2LimitDegrees=50, twistLimitDegrees=15}
 {bone1Name=dyn_hair_back_02_01_l, bone2Name=head, swing1LimitDegrees=25, swing2LimitDegrees=25, twistLimitDegrees=5}
 {bone1Name=dyn_hair_back_02_02_l, bone2Name=dyn_hair_back_02_01_l, swing1LimitDegrees=25, swing2LimitDegrees=25, twistLimitDegrees=15}
 {bone1Name=dyn_hair_back_02_03_l, bone2Name=dyn_hair_back_02_02_l, swing1LimitDegrees=15, swing2LimitDegrees=25, twistLimitDegrees=15}
 {bone1Name=dyn_hair_back_02_04_l, bone2Name=dyn_hair_back_02_03_l, swing1LimitDegrees=25, swing2LimitDegrees=35, twistLimitDegrees=15}
 {bone1Name=dyn_hair_back_02_05_l, bone2Name=dyn_hair_back_02_04_l, swing1LimitDegrees=35, swing2LimitDegrees=35, twistLimitDegrees=15}
 {bone1Name=dyn_hair_back_02_06_l, bone2Name=dyn_hair_back_02_05_l, swing1LimitDegrees=35, swing2LimitDegrees=50, twistLimitDegrees=15}
 {bone1Name=dyn_hair_back_02_01_r, bone2Name=head, swing1LimitDegrees=25, swing2LimitDegrees=25, twistLimitDegrees=5}
 {bone1Name=dyn_hair_back_02_03_r, bone2Name=dyn_hair_back_02_02_r, swing1LimitDegrees=15, swing2LimitDegrees=25, twistLimitDegrees=15}
 {bone1Name=dyn_hair_back_02_04_r, bone2Name=dyn_hair_back_02_03_r, swing1LimitDegrees=25, swing2LimitDegrees=35, twistLimitDegrees=15}
 {bone1Name=dyn_hair_back_02_05_r, bone2Name=dyn_hair_back_02_04_r, swing1LimitDegrees=35, swing2LimitDegrees=35, twistLimitDegrees=15}
 {bone1Name=dyn_hair_back_02_06_r, bone2Name=dyn_hair_back_02_05_r, swing1LimitDegrees=35, swing2LimitDegrees=50, twistLimitDegrees=15}
 {
  bone1Name=head
  bone2Name=neck_01
  swing1LimitDegrees=45
  swing1Motion=Free
  swing2LimitDegrees=45
  swing2Motion=Free
  twistLimitDegrees=45
  twistMotion=Free
 }
 {
  bone1Name=spine_03
  bone2Name=spine_02
  swing1LimitDegrees=45
  swing1Motion=Free
  swing2LimitDegrees=45
  swing2Motion=Free
  twistLimitDegrees=45
  twistMotion=Free
 }
 {
  bone1Name=spine_04
  bone2Name=spine_03
  swing1LimitDegrees=45
  swing1Motion=Free
  swing2LimitDegrees=45
  swing2Motion=Free
  twistLimitDegrees=45
  twistMotion=Free
 }
 {
  bone1Name=spine_05
  bone2Name=spine_04
  swing1LimitDegrees=45
  swing1Motion=Free
  swing2LimitDegrees=45
  swing2Motion=Free
  twistLimitDegrees=45
  twistMotion=Free
 }
 {bone1Name=dyn_hair_back_02_05_r, bone2Name=dyn_hair_back_01_05_r, swing1LimitDegrees=35, swing2LimitDegrees=35, twistLimitDegrees=35}
 {bone1Name=dyn_hair_back_02_05_l, bone2Name=dyn_hair_back_01_05_l, swing1LimitDegrees=35, swing2LimitDegrees=35, twistLimitDegrees=35}
 {bone1Name=dyn_hair_back_02_05_l, bone2Name=dyn_hair_back_02_05_r, swing1LimitDegrees=35, swing2LimitDegrees=35, twistLimitDegrees=35}
 {bone1Name=dyn_hair_back_02_06_r, bone2Name=dyn_hair_back_01_06_r, swing1LimitDegrees=35, swing2LimitDegrees=35, twistLimitDegrees=35}
 {bone1Name=dyn_hair_back_02_06_l, bone2Name=dyn_hair_back_01_06_l, swing1LimitDegrees=35, swing2LimitDegrees=35, twistLimitDegrees=35}
 {
  bone1Name=upperarm_l
  bone2Name=spine_05
  swing1LimitDegrees=45
  swing1Motion=Free
  swing2LimitDegrees=45
  swing2Motion=Free
  twistLimitDegrees=45
  twistMotion=Free
 }
 {
  bone1Name=upperarm_r
  bone2Name=spine_05
  swing1LimitDegrees=45
  swing1Motion=Free
  swing2LimitDegrees=45
  swing2Motion=Free
  twistLimitDegrees=45
  twistMotion=Free
 }
 {bone1Name=dyn_hair_back_02_06_l, bone2Name=dyn_hair_back_02_06_r, swing1LimitDegrees=35, swing2LimitDegrees=35, twistLimitDegrees=35}
 {bone1Name=dyn_hair_back_02_04_l, bone2Name=dyn_hair_back_02_04_r, swing1LimitDegrees=35, swing2LimitDegrees=35, twistLimitDegrees=35}
 {bone1Name=dyn_hair_back_02_03_r, bone2Name=dyn_hair_back_01_03_r, swing1LimitDegrees=35, swing2LimitDegrees=35, twistLimitDegrees=35}
 {bone1Name=dyn_hair_back_02_03_l, bone2Name=dyn_hair_back_01_03_l, swing1LimitDegrees=35, swing2LimitDegrees=35, twistLimitDegrees=35}
 {bone1Name=spine_05, bone2Name=dyn_hair_back_02_03_r, swing1LimitDegrees=45, swing2LimitDegrees=45, twistLimitDegrees=45}
 {bone1Name=spine_05, bone2Name=dyn_hair_back_02_03_l, swing1LimitDegrees=45, swing2LimitDegrees=45, twistLimitDegrees=45}
 {bone1Name=spine_05, bone2Name=dyn_hair_back_01_03_r, swing1LimitDegrees=45, swing2LimitDegrees=45, twistLimitDegrees=45}
 {bone1Name=spine_05, bone2Name=dyn_hair_back_01_02_r, swing1LimitDegrees=45, swing2LimitDegrees=45, twistLimitDegrees=45}
 {bone1Name=spine_05, bone2Name=dyn_hair_back_01_03_l, swing1LimitDegrees=45, swing2LimitDegrees=45, twistLimitDegrees=45}
 {bone1Name=dyn_hair_back_02_01_l, bone2Name=dyn_hair_back_02_01_r, swing1LimitDegrees=25, swing2LimitDegrees=25, twistLimitDegrees=25}
 {bone1Name=dyn_hair_back_02_01_l, bone2Name=dyn_hair_back_01_01_l, swing1LimitDegrees=35, swing2LimitDegrees=35, twistLimitDegrees=35}
 {bone1Name=dyn_hair_back_02_01_r, bone2Name=dyn_hair_back_01_01_r, swing1LimitDegrees=35, swing2LimitDegrees=35, twistLimitDegrees=35}
 {
  bone1Name=neck_01
  bone2Name=spine_05
  swing1LimitDegrees=45
  swing1Motion=Free
  swing2LimitDegrees=45
  swing2Motion=Free
  twistLimitDegrees=45
  twistMotion=Free
 }
 {bone1Name=spine_05, bone2Name=dyn_hair_back_01_02_l, swing1LimitDegrees=45, swing2LimitDegrees=45, twistLimitDegrees=45}
 {bone1Name=dyn_hair_back_02_02_r, bone2Name=dyn_hair_back_02_01_r, swing1LimitDegrees=25, swing2LimitDegrees=25, twistLimitDegrees=15}
 {bone1Name=spine_05, bone2Name=dyn_hair_back_02_02_l, swing1LimitDegrees=45, swing2LimitDegrees=45, twistLimitDegrees=45}
 {bone1Name=spine_05, bone2Name=dyn_hair_back_02_02_r, swing1LimitDegrees=45, swing2LimitDegrees=45, twistLimitDegrees=45}
 {bone1Name=dyn_hair_pigtail_06, bone2Name=dyn_hair_pigtail_05, swing1LimitDegrees=15, swing2LimitDegrees=15, twistLimitDegrees=15}
 {
  bone1Name=dyn_hair_pigtail_base
  bone2Name=head
  swing1LimitDegrees=20
  swing1Motion=Free
  swing2LimitDegrees=30
  swing2Motion=Free
  twistLimitDegrees=3
  twistMotion=Free
 }
 {bone1Name=dyn_hair_pigtail_01, bone2Name=dyn_hair_pigtail_base, swing1LimitDegrees=75, swing2LimitDegrees=75, twistLimitDegrees=25}
 {bone1Name=dyn_hair_pigtail_02, bone2Name=dyn_hair_pigtail_01, swing1LimitDegrees=30, swing2LimitDegrees=30, twistLimitDegrees=20}
 {bone1Name=dyn_hair_pigtail_03, bone2Name=dyn_hair_pigtail_02, swing1LimitDegrees=30, swing2LimitDegrees=30, twistLimitDegrees=20}
 {bone1Name=dyn_hair_pigtail_04, bone2Name=dyn_hair_pigtail_03, swing1LimitDegrees=30, swing2LimitDegrees=30, twistLimitDegrees=20}
 {bone1Name=dyn_hair_pigtail_05, bone2Name=dyn_hair_pigtail_04, swing1LimitDegrees=25, swing2LimitDegrees=25, twistLimitDegrees=25}
 {bone1Name=dyn_hair_back_02_06_r, bone2Name=dyn_hair_pigtail_06, swing1LimitDegrees=5, swing2LimitDegrees=15, twistLimitDegrees=15}
 {bone1Name=dyn_hair_back_02_05_r, bone2Name=dyn_hair_pigtail_05, swing1LimitDegrees=5, swing2LimitDegrees=15, twistLimitDegrees=15}
 {bone1Name=dyn_hair_back_02_04_r, bone2Name=dyn_hair_pigtail_04, swing1LimitDegrees=5, swing2LimitDegrees=15, twistLimitDegrees=15}
 {bone1Name=dyn_hair_back_02_03_r, bone2Name=dyn_hair_pigtail_03, swing1LimitDegrees=5, swing2LimitDegrees=15, twistLimitDegrees=15}
 {bone1Name=dyn_hair_back_02_02_r, bone2Name=dyn_hair_pigtail_02, swing1LimitDegrees=15, swing2LimitDegrees=15, twistLimitDegrees=15}
 {bone1Name=dyn_hair_back_02_01_r, bone2Name=dyn_hair_pigtail_01, swing1LimitDegrees=75, swing2LimitDegrees=75, twistLimitDegrees=75}
 {
  bone1Name=spine_02
  bone2Name=pelvis
  swing1LimitDegrees=45
  swing1Motion=Free
  swing2LimitDegrees=45
  swing2Motion=Free
  twistLimitDegrees=45
  twistMotion=Free
 }
]
