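# Animation Blueprint for the WrenchPeel female FaceAcc part, adding secondary motion to hair, pigtail and banana bones.
# It copies the parent mesh pose through a Control Rig, then selects among paired head/pigtail RigidBody simulations by backbling and front-end state.
# AnimDynamics then refines banana and hair bones before output; separate pigtail constraint chains are unconnected.
# Update logic computes gravity override and detects backblings or capes.
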AnimBlueprint /BRCosmetics/Characters/Player/Female/Medium/Bodies/F_MED_WrenchPeel/Meshes/Parts/F_MED_WrenchPeel_FaceAcc_AnimBP
target_skeleton=&/F_MED_WrenchPeel_FaceAcc_Skeleton
graphs [
 AnimGraph {nodes [
   n1 = "Constraint - dyn_hair_pigtail_01 (P)"(Alpha=1, ConstraintWeights_0=1) {
    alphaBoolBlend {blendOption=Linear}
    boneToModify {boneName=dyn_hair_pigtail_01}
    constraintSetup [{targetBone {boneName=head}}]
    constraintWeights [1]
   }
   n2 = "Constraint - dyn_hair_pigtail_01 (RR)"!(ComponentPose=@n1, Alpha=1, ConstraintWeights_0=0.5, ConstraintWeights_1:float=0.5) -> (Pose) {
    alphaBoolBlend {blendOption=Linear}
    boneToModify {boneName=dyn_hair_pigtail_01}
    constraintSetup [{transformType=Rotation, targetBone {boneName=dyn_hair_back_02_01_l}}, {transformType=Rotation, targetBone {boneName=dyn_hair_back_02_01_r}}]
    constraintWeights [1,1]
   }
   n3 = "Constraint - dyn_hair_pigtail_02 (PP)"!(ComponentPose=@n2, Alpha=1, ConstraintWeights_0=0.5, ConstraintWeights_1:float=0.5) -> (Pose) {
    alphaBoolBlend {blendOption=Linear}
    boneToModify {boneName=dyn_hair_pigtail_02}
    constraintSetup [{targetBone {boneName=dyn_hair_back_02_02_l}}, {targetBone {boneName=dyn_hair_back_02_02_r}}]
    constraintWeights [1,1]
   }
   n4 = "Constraint - dyn_hair_pigtail_03 (PP)"!(ComponentPose=@n3, Alpha=1, ConstraintWeights_0=0.5, ConstraintWeights_1:float=0.5) -> (Pose) {
    alphaBoolBlend {blendOption=Linear}
    boneToModify {boneName=dyn_hair_pigtail_03}
    constraintSetup [{targetBone {boneName=dyn_hair_back_02_03_l}}, {targetBone {boneName=dyn_hair_back_02_03_r}}]
    constraintWeights [1,1]
   }
   n5 = "Constraint - dyn_hair_pigtail_04 (PP)"!(ComponentPose=@n4, Alpha=1, ConstraintWeights_0=0.5, ConstraintWeights_1:float=0.5) -> (Pose) {
    alphaBoolBlend {blendOption=Linear}
    boneToModify {boneName=dyn_hair_pigtail_04}
    constraintSetup [{targetBone {boneName=dyn_hair_back_02_04_l}}, {targetBone {boneName=dyn_hair_back_02_04_r}}]
    constraintWeights [1,1]
   }
   n6 = "Constraint - dyn_hair_pigtail_05 (PP)"!(ComponentPose=@n5, Alpha=1, ConstraintWeights_0=0.5, ConstraintWeights_1:float=0.5) -> (Pose) {
    alphaBoolBlend {blendOption=Linear}
    boneToModify {boneName=dyn_hair_pigtail_05}
    constraintSetup [{targetBone {boneName=dyn_hair_back_02_05_l}}, {targetBone {boneName=dyn_hair_back_02_05_r}}]
    constraintWeights [1,1]
   }
   n7 = "Constraint - dyn_hair_pigtail_06 (PP)"!(ComponentPose=@n6, Alpha=1, ConstraintWeights_0=0.5, ConstraintWeights_1:float=0.5) -> (Pose) {
    alphaBoolBlend {blendOption=Linear}
    boneToModify {boneName=dyn_hair_pigtail_06}
    constraintSetup [{targetBone {boneName=dyn_hair_back_02_06_l}}, {targetBone {boneName=dyn_hair_back_02_06_r}}]
    constraintWeights [1,1]
   }
   n8 = Anim Dynamics - Bone: dyn_hair_pigtail_01!(ComponentPose=@n7, Alpha=1, GravityOverride:Vector="0,0,0", ExternalForce="0,0,0") -> (Pose) {
    angularDampingOverride=1
    angularSpringConstant=1000
    bAngularSpring=true
    bOverrideAngularDamping=true
    bUseGravityOverride=true
    simulationSpace=RootRelative
    alphaBoolBlend {blendOption=Linear}
    boundBone {boneName=dyn_hair_pigtail_01}
    componentAppliedLinearAccClamp=<5000 5000 5000>
    componentLinearAccScale=<0.05 0.05 0.05>
    componentLinearVelScale=<0.05 0.05 0.05>
    physicsBodyDefinitions [{
      boundBone {boneName=dyn_hair_pigtail_01}
      boxExtents=<10 10 10>
      constraintSetup {angularLimitsMax=<0 0 5>, angularLimitsMin=<0 -20 -5>, angularTarget=<1 0 0>}
      localJointOffset=<0.4 0 0>
     }]
    simSpaceSettings {maxAngularAcceleration=100, maxAngularVelocity=100, simSpaceAngularAlpha=0.2}
   }
   n9 = Anim Dynamics - Bone: dyn_hair_pigtail_02!(ComponentPose=@n8, Alpha=1, GravityOverride:Vector="0,0,0", ExternalForce="0,0,0") -> (Pose) {
    angularDampingOverride=1
    angularSpringConstant=1000
    bAngularSpring=true
    bOverrideAngularDamping=true
    bUseGravityOverride=true
    simulationSpace=RootRelative
    alphaBoolBlend {blendOption=Linear}
    boundBone {boneName=dyn_hair_pigtail_02}
    componentAppliedLinearAccClamp=<5000 5000 5000>
    componentLinearAccScale=<0.05 0.05 0.05>
    componentLinearVelScale=<0.05 0.05 0.05>
    physicsBodyDefinitions [{
      boundBone {boneName=dyn_hair_pigtail_02}
      boxExtents=<10 10 10>
      constraintSetup {angularLimitsMax=<0 0 5>, angularLimitsMin=<0 -20 -5>, angularTarget=<1 0 0>}
      localJointOffset=<0.4 0 0>
     }]
    simSpaceSettings {maxAngularAcceleration=100, maxAngularVelocity=100, simSpaceAngularAlpha=0.2}
   }
   n10 = Anim Dynamics - Bone: dyn_hair_pigtail_03!(ComponentPose=@n9, Alpha=1, GravityOverride:Vector="0,0,0", ExternalForce="0,0,0") -> (Pose) {
    angularDampingOverride=1
    angularSpringConstant=1000
    bAngularSpring=true
    bOverrideAngularDamping=true
    bUseGravityOverride=true
    simulationSpace=RootRelative
    alphaBoolBlend {blendOption=Linear}
    boundBone {boneName=dyn_hair_pigtail_03}
    componentAppliedLinearAccClamp=<5000 5000 5000>
    componentLinearAccScale=<0.05 0.05 0.05>
    componentLinearVelScale=<0.05 0.05 0.05>
    physicsBodyDefinitions [{
      boundBone {boneName=dyn_hair_pigtail_03}
      boxExtents=<10 10 10>
      constraintSetup {angularLimitsMax=<0 0 5>, angularLimitsMin=<0 -20 -5>, angularTarget=<1 0 0>}
      localJointOffset=<0.4 0 0>
     }]
    simSpaceSettings {maxAngularAcceleration=100, maxAngularVelocity=100, simSpaceAngularAlpha=0.2}
   }
   n11 = Anim Dynamics - Bone: dyn_hair_pigtail_04!(ComponentPose=@n10, Alpha=1, GravityOverride:Vector="0,0,0", ExternalForce="0,0,0") -> (Pose) {
    angularDampingOverride=1
    angularSpringConstant=1000
    bAngularSpring=true
    bOverrideAngularDamping=true
    bUseGravityOverride=true
    simulationSpace=RootRelative
    alphaBoolBlend {blendOption=Linear}
    boundBone {boneName=dyn_hair_pigtail_04}
    componentAppliedLinearAccClamp=<5000 5000 5000>
    componentLinearAccScale=<0.05 0.05 0.05>
    componentLinearVelScale=<0.05 0.05 0.05>
    physicsBodyDefinitions [{
      boundBone {boneName=dyn_hair_pigtail_04}
      boxExtents=<10 10 10>
      constraintSetup {angularLimitsMax=<0 0 5>, angularLimitsMin=<0 -20 -5>, angularTarget=<1 0 0>}
      localJointOffset=<0.4 0 0>
     }]
    simSpaceSettings {maxAngularAcceleration=100, maxAngularVelocity=100, simSpaceAngularAlpha=0.2}
   }
   n12 = Anim Dynamics - Bone: dyn_hair_pigtail_05!(ComponentPose=@n11, Alpha=1, GravityOverride:Vector="0,0,0", ExternalForce="0,0,0") -> (Pose) {
    angularDampingOverride=1
    angularSpringConstant=1000
    bAngularSpring=true
    bOverrideAngularDamping=true
    bUseGravityOverride=true
    simulationSpace=RootRelative
    alphaBoolBlend {blendOption=Linear}
    boundBone {boneName=dyn_hair_pigtail_05}
    componentAppliedLinearAccClamp=<5000 5000 5000>
    componentLinearAccScale=<0.05 0.05 0.05>
    componentLinearVelScale=<0.05 0.05 0.05>
    physicsBodyDefinitions [{
      boundBone {boneName=dyn_hair_pigtail_05}
      boxExtents=<10 10 10>
      constraintSetup {angularLimitsMax=<0 0 5>, angularLimitsMin=<0 -20 -5>, angularTarget=<1 0 0>}
      localJointOffset=<0.4 0 0>
     }]
    simSpaceSettings {maxAngularAcceleration=100, maxAngularVelocity=100, simSpaceAngularAlpha=0.2}
   }
   Anim Dynamics - Bone: dyn_hair_pigtail_06!(ComponentPose=@n12, Alpha=1, GravityOverride:Vector="0,0,0", ExternalForce="0,0,0") -> (Pose) {
    angularDampingOverride=1
    angularSpringConstant=1000
    bAngularSpring=true
    bOverrideAngularDamping=true
    bUseGravityOverride=true
    simulationSpace=RootRelative
    alphaBoolBlend {blendOption=Linear}
    boundBone {boneName=dyn_hair_pigtail_06}
    componentAppliedLinearAccClamp=<5000 5000 5000>
    componentLinearAccScale=<0.05 0.05 0.05>
    componentLinearVelScale=<0.05 0.05 0.05>
    physicsBodyDefinitions [{
      boundBone {boneName=dyn_hair_pigtail_06}
      boxExtents=<10 10 10>
      constraintSetup {angularLimitsMax=<0 0 5>, angularLimitsMin=<0 -20 -5>, angularTarget=<1 0 0>}
      localJointOffset=<0.4 0 0>
     }]
    simSpaceSettings {maxAngularAcceleration=100, maxAngularVelocity=100, simSpaceAngularAlpha=0.2}
   }
   "dyn_hair_pigtail_01.rotateY = dyn_hair_bunch_base.rotateY * -1 - Bone Driven Controller"!(ComponentPose:Component Space Pose Link Structure, Alpha:float=1) -> (Pose:Component Space Pose Link Structure) {
    is=title AnimGraphNode_BoneDrivenController
    bAffectTargetRotationY=true
    modificationMode=ReplaceComponent
    multiplier=-1
    sourceComponent=RotationY
    sourceBone {boneName=dyn_hair_bunch_base}
    targetBone {boneName=dyn_hair_pigtail_01}
   }
   n13 = "Constraint - dyn_hair_bunch_base (P)"(Alpha=1, ConstraintWeights_0=1) {
    alphaBoolBlend {blendOption=Linear}
    boneToModify {boneName=dyn_hair_bunch_base}
    constraintSetup [{targetBone {boneName=head}}]
    constraintWeights [1]
   }
   "Constraint - dyn_hair_bunch_base (R)"(ComponentPose=@n13, Alpha=0.9, ConstraintWeights_0=1) {
    alphaBoolBlend {blendOption=Linear}
    boneToModify {boneName=dyn_hair_bunch_base}
    constraintSetup [{transformType=Rotation, targetBone {boneName=dyn_hair_pigtail_01}}]
    constraintWeights [1]
   }
   n14 = Anim Dynamics - Bone: dyn_banana_01(ComponentPose=Local To Component(LocalPose=Use cached pose 'PreSim2'()), Alpha=1, ExternalForce="0,0,0") {
    angularDampingOverride=1
    angularSpringConstant=1000
    boundBone {boneName=dyn_banana_01}
    componentAppliedLinearAccClamp=<5000 5000 5000>
    physicsBodyDefinitions [{
      boundBone {boneName=dyn_banana_01}
      boxExtents=<10 10 10>
      constraintSetup {angularLimitsMax=<0.5 0.5 0.5>, angularLimitsMin=<-0.5 -0.5 -0.5>, angularTarget=<0 0 0>}
      localJointOffset=<0.1 0 0>
     }]
    simSpaceSettings {maxAngularAcceleration=100, maxAngularVelocity=100}
   }
   n15 = Anim Dynamics - Bone: dyn_hair_tailbunch_mid!(ComponentPose=@n14, Alpha=1, GravityOverride:Vector="0,0,0", ExternalForce="0,0,0") -> (Pose) {
    angularDampingOverride=1
    angularSpringConstant=1000
    bAngularSpring=true
    bOverrideAngularDamping=true
    bUseGravityOverride=true
    simulationSpace=BoneRelative
    alphaBoolBlend {blendOption=Linear}
    boundBone {boneName=dyn_hair_tailbunch_mid}
    componentAppliedLinearAccClamp=<5000 5000 5000>
    componentLinearAccScale=<0.05 0.05 0.05>
    componentLinearVelScale=<0.05 0.05 0.05>
    physicsBodyDefinitions [{
      boundBone {boneName=dyn_hair_tailbunch_mid}
      boxExtents=<10 10 10>
      constraintSetup {angularLimitsMax=<0 7 7>, angularLimitsMin=<0 -20 -7>, angularTarget=<1 0 0>}
      localJointOffset=<0.1 0 0>
     }]
    relativeSpaceBone {boneName=head}
    simSpaceSettings {maxAngularAcceleration=100, maxAngularVelocity=100, simSpaceAngularAlpha=0.2}
   }
   n16 = Anim Dynamics - Bone: dyn_hair_tailbunch_r!(ComponentPose=@n15, Alpha=1, GravityOverride:Vector="0,0,0", ExternalForce="0,0,0") -> (Pose) {
    angularDampingOverride=1
    angularSpringConstant=1000
    bAngularSpring=true
    bOverrideAngularDamping=true
    bUseGravityOverride=true
    simulationSpace=BoneRelative
    alphaBoolBlend {blendOption=Linear}
    boundBone {boneName=dyn_hair_tailbunch_r}
    componentAppliedLinearAccClamp=<5000 5000 5000>
    componentLinearAccScale=<0.05 0.05 0.05>
    componentLinearVelScale=<0.05 0.05 0.05>
    physicsBodyDefinitions [{
      boundBone {boneName=dyn_hair_tailbunch_r}
      boxExtents=<10 10 10>
      constraintSetup {angularLimitsMax=<1 16 5>, angularLimitsMin=<-1 -16 -17>, angularTarget=<1 0 0>}
      localJointOffset=<0.2 0 0>
     }]
    relativeSpaceBone {boneName=head}
    simSpaceSettings {maxAngularAcceleration=100, maxAngularVelocity=100, simSpaceAngularAlpha=0.2}
   }
   n17 = Anim Dynamics - Bone: dyn_hair_tailbunch_l!(ComponentPose=@n16, Alpha=1, GravityOverride:Vector="0,0,0", ExternalForce="0,0,0") -> (Pose) {
    angularDampingOverride=1
    angularSpringConstant=1000
    bAngularSpring=true
    bOverrideAngularDamping=true
    bUseGravityOverride=true
    simulationSpace=BoneRelative
    alphaBoolBlend {blendOption=Linear}
    boundBone {boneName=dyn_hair_tailbunch_l}
    componentAppliedLinearAccClamp=<5000 5000 5000>
    componentLinearAccScale=<0.05 0.05 0.05>
    componentLinearVelScale=<0.05 0.05 0.05>
    physicsBodyDefinitions [{
      boundBone {boneName=dyn_hair_tailbunch_l}
      boxExtents=<10 10 10>
      constraintSetup {angularLimitsMax=<1 15 17>, angularLimitsMin=<-1 -16 -7>, angularTarget=<1 0 0>}
      localJointOffset=<0.6 0 0>
     }]
    relativeSpaceBone {boneName=head}
    simSpaceSettings {maxAngularAcceleration=100, maxAngularVelocity=100, simSpaceAngularAlpha=0.2}
   }
   n18 = Anim Dynamics - Bone: dyn_hair_bunch_01!(ComponentPose=@n17, Alpha=1, GravityOverride:Vector="0,0,0", ExternalForce="0,0,0") -> (Pose) {
    angularDampingOverride=1
    angularSpringConstant=1000
    bAngularSpring=true
    bOverrideAngularDamping=true
    bUseGravityOverride=true
    simulationSpace=BoneRelative
    alphaBoolBlend {blendOption=Linear}
    boundBone {boneName=dyn_hair_bunch_01}
    componentAppliedLinearAccClamp=<5000 5000 5000>
    componentLinearAccScale=<0.05 0.05 0.05>
    componentLinearVelScale=<0.05 0.05 0.05>
    physicsBodyDefinitions [{
      boundBone {boneName=dyn_hair_bunch_01}
      boxExtents=<25 25 25>
      constraintSetup {angularLimitsMax=<2 5 10>, angularLimitsMin=<-2 -5 -10>, angularTarget=<1 0 0>}
      localJointOffset=<1 0 0>
     }]
    relativeSpaceBone {boneName=head}
    simSpaceSettings {maxAngularAcceleration=100, maxAngularVelocity=100, simSpaceAngularAlpha=0.2}
   }
   n19 = Anim Dynamics - Bone: dyn_hair_01_fr_l!(ComponentPose=@n18, Alpha=1, GravityOverride:Vector="0,0,0", ExternalForce="0,0,0") -> (Pose) {
    angularDampingOverride=1
    angularSpringConstant=1000
    bAngularSpring=true
    bOverrideAngularDamping=true
    bUseGravityOverride=true
    simulationSpace=World
    alphaBoolBlend {blendOption=Linear}
    boundBone {boneName=dyn_hair_01_fr_l}
    componentAppliedLinearAccClamp=<5000 5000 5000>
    componentLinearAccScale=<0.05 0.05 0.05>
    componentLinearVelScale=<0.05 0.05 0.05>
    physicsBodyDefinitions [{
      boundBone {boneName=dyn_hair_01_fr_l}
      boxExtents=<25 25 25>
      constraintSetup {angularLimitsMax=<2 5 8>, angularLimitsMin=<-2 -15 -2>, angularTarget=<1 0 0>}
      localJointOffset=<0.1 0 0>
     }]
    relativeSpaceBone {boneName=spine_05}
    simSpaceSettings {maxAngularAcceleration=100, maxAngularVelocity=100, simSpaceAngularAlpha=0.2}
   }
   n20 = Anim Dynamics - Bone: dyn_hair_01_fr_r!(ComponentPose=@n19, Alpha=1, GravityOverride:Vector="0,0,0", ExternalForce="0,0,0") -> (Pose) {
    angularDampingOverride=1
    angularSpringConstant=1000
    bAngularSpring=true
    bOverrideAngularDamping=true
    bUseGravityOverride=true
    simulationSpace=World
    alphaBoolBlend {blendOption=Linear}
    boundBone {boneName=dyn_hair_01_fr_r}
    componentAppliedLinearAccClamp=<5000 5000 5000>
    componentLinearAccScale=<0.05 0.05 0.05>
    componentLinearVelScale=<0.05 0.05 0.05>
    physicsBodyDefinitions [{
      boundBone {boneName=dyn_hair_01_fr_r}
      boxExtents=<25 25 25>
      constraintSetup {angularLimitsMax=<2 8 3>, angularLimitsMin=<-2 -15 -7>, angularTarget=<1 0 0>}
      localJointOffset=<0.1 0 0>
     }]
    relativeSpaceBone {boneName=spine_05}
    simSpaceSettings {maxAngularAcceleration=100, maxAngularVelocity=100, simSpaceAngularAlpha=0.2}
   }
   n21 = Anim Dynamics - Bone: dyn_hair_01_side_l!(ComponentPose=@n20, Alpha=1, GravityOverride:Vector="0,0,0", ExternalForce="0,0,0") -> (Pose) {
    angularDampingOverride=1
    angularSpringConstant=1000
    bAngularSpring=true
    bOverrideAngularDamping=true
    bUseGravityOverride=true
    simulationSpace=World
    alphaBoolBlend {blendOption=Linear}
    boundBone {boneName=dyn_hair_01_side_l}
    componentAppliedLinearAccClamp=<5000 5000 5000>
    componentLinearAccScale=<0.05 0.05 0.05>
    componentLinearVelScale=<0.05 0.05 0.05>
    physicsBodyDefinitions [{
      boundBone {boneName=dyn_hair_01_side_l}
      boxExtents=<25 25 25>
      constraintSetup {angularLimitsMax=<0.5 3 2>, angularLimitsMin=<-0.5 -13 -2>, angularTarget=<1 0 0>}
      localJointOffset=<0.1 0 0>
     }]
    relativeSpaceBone {boneName=spine_05}
    simSpaceSettings {maxAngularAcceleration=100, maxAngularVelocity=100, simSpaceAngularAlpha=0.2}
   }
   n22 = Anim Dynamics - Bone: dyn_hair_01_side_r!(ComponentPose=@n21, Alpha=1, GravityOverride:Vector="0,0,0", ExternalForce="0,0,0") -> (Pose) {
    angularDampingOverride=1
    angularSpringConstant=1000
    bAngularSpring=true
    bOverrideAngularDamping=true
    bUseGravityOverride=true
    simulationSpace=World
    alphaBoolBlend {blendOption=Linear}
    boundBone {boneName=dyn_hair_01_side_r}
    componentAppliedLinearAccClamp=<5000 5000 5000>
    componentLinearAccScale=<0.05 0.05 0.05>
    componentLinearVelScale=<0.05 0.05 0.05>
    physicsBodyDefinitions [{
      boundBone {boneName=dyn_hair_01_side_r}
      boxExtents=<25 25 25>
      constraintSetup {angularLimitsMax=<0.5 2 2>, angularLimitsMin=<-0.5 -13 -2>, angularTarget=<1 0 0>}
      localJointOffset=<0.1 0 0>
     }]
    relativeSpaceBone {boneName=spine_05}
    simSpaceSettings {maxAngularAcceleration=100, maxAngularVelocity=100, simSpaceAngularAlpha=0.2}
   }
   Output Pose(Result=Component To Local(ComponentPose=@n22)) {layerGroup=None}
   n23 = Control Rig(Source=Copy Pose From Mesh() {bUseAttachedParent=true}) {controlRigAssetReference {blueprintRigClass=&/F_MED_WrenchPeel_FaceAcc_CtrlRig.F_MED_WrenchPeel_FaceAcc_CtrlRig_C}}
   Save cached pose 'PreSim'(Pose=@n23) {cacheName=PreSim}
   n24 = RigidBody(ComponentPose=Local To Component(LocalPose=Use cached pose 'PreSim'())) {
    bClampLinearTranslationLimitToRefPose=true
    bOverrideWorldGravity=true
    overridePhysicsAsset=&/F_MED_WrenchPeel_FaceAcc_Physics
    alphaBoolBlend {blendOption=Linear}
    baseBoneRef {boneName=neck_02}
    componentAppliedLinearAccClamp=<1000 1000 1000>
    overrideWorldGravity {z=-980}
    simSpaceSettings {
     maxAngularAcceleration=200
     maxAngularVelocity=3
     maxLinearAcceleration=100
     maxLinearVelocity=70
     worldAlpha=0.5
     externalLinearDragV=<0.7 0.4 0.4>
     externalLinearVelocity=<2 0.2 0.2>
    }
   }
   n25 = RigidBody(ComponentPose=@n24) {
    bClampLinearTranslationLimitToRefPose=true
    bOverrideWorldGravity=true
    overridePhysicsAsset=&/F_MED_WrenchPeel_FaceAcc_Pigtail_Physics
    alphaBoolBlend {blendOption=Linear}
    baseBoneRef {boneName=dyn_hair_pigtail_base}
    componentAppliedLinearAccClamp=<1000 1000 1000>
    overrideWorldGravity {z=-980}
    simSpaceSettings {
     maxAngularAcceleration=200
     maxAngularVelocity=3
     maxLinearAcceleration=100
     maxLinearVelocity=70
     worldAlpha=0.3
     externalLinearDragV=<0.7 0.4 0.4>
     externalLinearVelocity=<2 0.2 0.2>
    }
   }
   n26 = Component To Local(ComponentPose=@n25)
   n27 = RigidBody!(ComponentPose=Local To Component(LocalPose=Use cached pose 'PreSim'()), Alpha, OverrideWorldGravity:Vector="0,0,0", ExternalForce) -> (Pose) {
    bClampLinearTranslationLimitToRefPose=true
    bOverrideWorldGravity=true
    overridePhysicsAsset=&F_MED_WrenchPeel_FaceAcc_Physics
    simulationSpace=BaseBoneSpace
    alphaBoolBlend {blendOption=Linear}
    baseBoneRef {boneName=neck_02}
    simSpaceSettings {
     dampingAlpha=0.5
     maxAngularAcceleration=1000
     maxAngularVelocity=5
     maxLinearAcceleration=1050
     maxLinearVelocity=1000
     velocityScaleZ=0.5
     worldAlpha=1
     externalAngularVelocity=<0.1 0.1 0.1>
     externalLinearDragV=<0.3 0.3 0.5>
     externalLinearVelocity=<0.25 0.25 0.5>
    }
   }
   n28 = RigidBody(ComponentPose=@n27, OverrideWorldGravity="0,0,0") {
    bClampLinearTranslationLimitToRefPose=true
    bOverrideWorldGravity=true
    overridePhysicsAsset=&F_MED_WrenchPeel_FaceAcc_Pigtail_Physics
    simulationSpace=BaseBoneSpace
    alphaBoolBlend {blendOption=Linear}
    baseBoneRef {boneName=dyn_hair_pigtail_base}
    simSpaceSettings {
     dampingAlpha=0.5
     maxAngularAcceleration=500
     maxAngularVelocity=5
     maxLinearAcceleration=500
     maxLinearVelocity=900
     velocityScaleZ=0.5
     worldAlpha=1
     externalAngularVelocity=<0.1 0.1 0.1>
     externalLinearDragV=<0.3 0.3 0.5>
     externalLinearVelocity=<0.3 0.3 0.5>
    }
   }
   n29 = Component To Local(ComponentPose=@n28)
   n30 = Get bIsInFrontEnd()
   n31 = RigidBody!(ComponentPose=Local To Component(LocalPose=Use cached pose 'PreSim'()), Alpha, ExternalForce) -> (Pose) {
    bClampLinearTranslationLimitToRefPose=true
    bOverrideWorldGravity=true
    overridePhysicsAsset=&/F_MED_WrenchPeel_FaceAcc_Physics_Backbling
    simulationSpace=BaseBoneSpace
    alphaBoolBlend {blendOption=Linear}
    baseBoneRef {boneName=neck_02}
    overrideWorldGravity {z=-980}
    simSpaceSettings {maxAngularAcceleration=10, maxAngularVelocity=3, maxLinearAcceleration=10, maxLinearVelocity=10, velocityScaleZ=0, worldAlpha=0.15}
   }
   n32 = RigidBody!(ComponentPose=@n31, Alpha, OverrideWorldGravity:Vector="0,0,0", ExternalForce) -> (Pose) {
    bClampLinearTranslationLimitToRefPose=true
    bOverrideWorldGravity=true
    overridePhysicsAsset=&/F_MED_WrenchPeel_FaceAcc_Pigtail_Physics_Backbling
    simulationSpace=BaseBoneSpace
    alphaBoolBlend {blendOption=Linear}
    baseBoneRef {boneName=neck_02}
    simSpaceSettings {maxAngularAcceleration=5, maxAngularVelocity=3, maxLinearAcceleration=5, maxLinearVelocity=5, velocityScaleZ=0, worldAlpha=0.15}
   }
   Save cached pose 'PreSim2'(Pose=Blend Poses by bool(
     BlendPose_0=Component To Local(ComponentPose=@n32)
     BlendPose_1=Blend Poses by bool(BlendPose_0=@n26, BlendPose_1=@n29, bActiveValue=@n30) {blendType=Linear}
     bActiveValue=Get Wear_Backbling() -> (Wear_Backbling:bool)
    ) {blendType=Linear}) {cacheName=PreSim2}
  ]}
 EventGraph {nodes [
   Event BlueprintUpdateAnimation()
   > n1 = Sequence!(execute) -> (then_0, then_1, then_2:Exec)
   > n2 = Gravity Override State Function DEPRECATED(
    CharacterPartAnimInstance=Self-Reference()
    GameTimeInSeconds=GetGameTimeInSeconds()
    GravityOverrideParams=Get Gravity Override Params()
    RBAN Params=Get RBAN Params()
   )
   > Set GravityOverride(GravityOverride=@n2)
   n3 = CheckForBackbling(self=Self-Reference())
   Set Wear_cape(execute=@n1.then_1, Wear_cape:bool=@n3.bIsBackblingACape) -> (Output_Get:bool)
   Set Wear_Backbling(execute=@n1.then_2, Wear_Backbling:bool=@n3) -> (Output_Get:bool)
   Event BlueprintInitializeAnimation()
   > Set SkelMesh(SkelMesh=GetOwnerSkeletalMeshComponent())
   n4 = TryGetPawnOwner()
   > Is Valid(InputObject=@n4)
   > n5 = Cast To FortPlayerPawnAthena(Object=@n4)
   > Set PlayerPawn(PlayerPawn:Fort Player Pawn Athena Object Reference=@n5) -> (Output_Get:Fort Player Pawn Athena Object Reference)
  ]}
]
variables [
 Gravity Override Params {
  type=GravityOverrideParamsStruct
  default {
   dBNO {jointName=root, gravityOverride=<0 -450 -980>}
   skydive_DivingDown {windAmplitude=<50 100 100>}
   skydive_FloatingDown {gravityOverride=<0 -200 300>, windAmplitude=<150 -100 150>, windFrequency=<1.5 10 12>, windOffset=<0 -500 70>}
   skydive_GoingUp {gravityOverride=<0 0 -400>, windAmplitude=<50 -50 50>, windOffset=<0 -600 50>}
   skydive_Parachuting {gravityOverride=<0 -500 -980>, windAmplitude=<100 -100 100>}
  }
 }
 GravityOverride {type=Vector}
 PlayerPawn {type=FortPlayerPawnAthena}
 RBAN Params {
  type=RBANParamsStruct
  default {
   dBNO {simSpaceSettings {
     maxAngularAcceleration=1000
     maxAngularVelocity=3
     maxLinearAcceleration=300
     maxLinearVelocity=700
     velocityScaleZ=0
     externalLinearDragV=<0 0.5 0>
    }}
   dBNO_Carried {simSpaceSettings {
     maxAngularAcceleration=1000
     maxAngularVelocity=3
     maxLinearAcceleration=300
     maxLinearVelocity=700
     velocityScaleZ=0
     externalLinearDragV=<0 0.5 0>
    }}
   emoteOrMelee {simSpaceSettings {
     maxAngularAcceleration=1000
     maxAngularVelocity=3
     maxLinearAcceleration=300
     maxLinearVelocity=700
     velocityScaleZ=0
     externalLinearDragV=<0 0.5 0>
    }}
   front_End {simSpaceSettings {
     maxAngularAcceleration=1000
     maxAngularVelocity=3
     maxLinearAcceleration=300
     maxLinearVelocity=700
     velocityScaleZ=0
     externalLinearDragV=<0 0.5 0>
    }}
   onGround_Crouch_Moving {simSpaceSettings {
     maxAngularAcceleration=1000
     maxAngularVelocity=3
     maxLinearAcceleration=300
     maxLinearVelocity=700
     velocityScaleZ=0
     externalLinearDragV=<0 0.5 0>
    }}
   onGround_Falling {simSpaceSettings {
     maxAngularAcceleration=1000
     maxAngularVelocity=3
     maxLinearAcceleration=300
     maxLinearVelocity=700
     velocityScaleZ=0
     externalLinearDragV=<0 0.5 0>
    }}
   onGround_Moving {simSpaceSettings {
     maxAngularAcceleration=1000
     maxAngularVelocity=3
     maxLinearAcceleration=300
     maxLinearVelocity=700
     velocityScaleZ=0
     externalLinearDragV=<0 0.5 0>
    }}
   onGround_Standing {simSpaceSettings {
     maxAngularAcceleration=1000
     maxAngularVelocity=3
     maxLinearAcceleration=300
     maxLinearVelocity=700
     velocityScaleZ=0
     externalLinearDragV=<0 0.5 0>
    }}
   rail_Grinding {simSpaceSettings {
     maxAngularAcceleration=1000
     maxAngularVelocity=3
     maxLinearAcceleration=300
     maxLinearVelocity=700
     velocityScaleZ=0
     worldAlpha=0.1
     externalLinearDragV=<0 0.5 0>
    }}
   riding_Moving {simSpaceSettings {
     maxAngularAcceleration=1000
     maxAngularVelocity=3
     maxLinearAcceleration=300
     maxLinearVelocity=700
     velocityScaleZ=0
     externalLinearDragV=<0 0.5 0>
    }}
   riding_Moving_Targeting {simSpaceSettings {
     maxAngularAcceleration=1000
     maxAngularVelocity=3
     maxLinearAcceleration=300
     maxLinearVelocity=700
     velocityScaleZ=0
     externalLinearDragV=<0 0.5 0>
    }}
   riding_Standing {simSpaceSettings {
     maxAngularAcceleration=1000
     maxAngularVelocity=3
     maxLinearAcceleration=300
     maxLinearVelocity=700
     velocityScaleZ=0
     externalLinearDragV=<0 0.5 0>
    }}
   skydive_DivingDown {simSpaceSettings {
     maxAngularAcceleration=1000
     maxAngularVelocity=3
     maxLinearAcceleration=300
     maxLinearVelocity=700
     velocityScaleZ=0
     externalLinearDragV=<0 0.5 0>
    }}
   skydive_FloatingDown {
    componentLinearAccScale=<0 0.5 0>
    componentLinearVelScale=<0 0.5 0>
    simSpaceSettings {
     maxAngularAcceleration=1000
     maxAngularVelocity=3
     maxLinearAcceleration=300
     maxLinearVelocity=700
     velocityScaleZ=0
     externalLinearDragV=<0 0.5 0>
    }
   }
   skydive_GoingUp {simSpaceSettings {
     maxAngularAcceleration=1000
     maxAngularVelocity=3
     maxLinearAcceleration=300
     maxLinearVelocity=700
     velocityScaleZ=0
     externalLinearDragV=<0 0.5 0>
    }}
   skydive_Parachuting {
    componentLinearAccScale=<0 0.1 0.3>
    componentLinearVelScale=<0 0.1 0.3>
    simSpaceSettings {
     maxAngularAcceleration=1000
     maxAngularVelocity=3
     maxLinearAcceleration=300
     maxLinearVelocity=700
     velocityScaleZ=0
     externalLinearDragV=<0 0.5 0>
    }
   }
   skydive_Surfing {simSpaceSettings {
     maxAngularAcceleration=1000
     maxAngularVelocity=3
     maxLinearAcceleration=300
     maxLinearVelocity=700
     velocityScaleZ=0
     externalLinearDragV=<0 0.5 0>
    }}
   swimming {simSpaceSettings {
     maxAngularAcceleration=1000
     maxAngularVelocity=3
     maxLinearAcceleration=300
     maxLinearVelocity=700
     velocityScaleZ=0
     externalLinearDragV=<0 0.5 0>
    }}
   tactical_Sprint {simSpaceSettings {
     maxAngularAcceleration=1000
     maxAngularVelocity=3
     maxLinearAcceleration=300
     maxLinearVelocity=700
     velocityScaleZ=0
     externalLinearDragV=<0 0.5 0>
    }}
   ziplining {simSpaceSettings {
     maxAngularAcceleration=1000
     maxAngularVelocity=3
     maxLinearAcceleration=300
     maxLinearVelocity=700
     velocityScaleZ=0
     externalLinearDragV=<0 0.5 0>
    }}
  }
 }
 SkelMesh {type=SkeletalMeshComponent}
 Wear_Backbling {type=boolean}
 Wear_cape {type=boolean}
]
