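# Physics asset for WrenchPeel's pigtail face accessory (medium female body), simulating back-hair bones with body colliders, named as cape colliders.
# Four six-capsule chains (two rows, left and right) hang from spine_05, with swing limits generally widest toward the tips.
# Cross-constraints tie neighboring chains at several segments and upper segments back to spine_05, so they move like a linked sheet.
# Spine and upper-arm capsules form the body volume, joined by free-rotation constraints.
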PhysicsAsset /BRCosmetics/Characters/Player/Female/Medium/Bodies/F_MED_WrenchPeel/Meshes/Parts/F_MED_WrenchPeel_FaceAcc_Pigtail_CapeColiders
bodies {
 dyn_hair_back_01_01_l [{
   shapeType=Capsule
   length=0.4486
   radius=6.4171
   shapeName=dyn_hair_back_01_06_r_capsule
   center=<5.3392 -0.0003 -4.2209>
   rotation=<rot 82.7293 -2.1991 -0.0003>
  }]
 dyn_hair_back_01_01_r [{
   shapeType=Capsule
   length=0.4486
   radius=6.4171
   shapeName=dyn_hair_back_01_06_r_capsule
   center=<5.3392 0 -4.2209>
   rotation=<rot 82.7293 0.0001 0.0002>
  }]
 dyn_hair_back_01_02_l [{
   shapeType=Capsule
   length=3.0684
   radius=5.7206
   shapeName=dyn_hair_back_01_06_r_capsule
   center=<4.2068 1.8928 -3.3361>
   rotation=<rot 82.9408 -4.5339 -10.7191>
  }]
 dyn_hair_back_01_02_r [{
   shapeType=Capsule
   length=3.0684
   radius=5.7206
   shapeName=dyn_hair_back_01_06_r_capsule
   center=<3.7443 -2.7699 -3.957>
   rotation=<rot 87.3788 -4.5337 0.0003>
  }]
 dyn_hair_back_01_03_l [{
   shapeType=Capsule
   length=0.4486
   radius=6.4171
   shapeName=dyn_hair_back_01_06_r_capsule
   center=<3.9461 -0.1959 -4.2392>
   rotation=<rot 87.0745 -1.8755 -5.1397>
  }]
 dyn_hair_back_01_03_r [{
   shapeType=Capsule
   length=0.4486
   radius=6.4171
   shapeName=dyn_hair_back_01_06_r_capsule
   center=<3.9526 0.0016 -4.2392>
   rotation=<rot 87.0746 -1.8753 0.0002>
  }]
 dyn_hair_back_01_04_l [{
   shapeType=Capsule
   length=0.4486
   radius=6.4171
   shapeName=dyn_hair_back_01_06_r_capsule
   center=<3.9572 -0.1886 -3.8474>
   rotation=<rot 90 -0.0011 0>
  }]
 dyn_hair_back_01_04_r [{
   shapeType=Capsule
   length=0.4486
   radius=6.4171
   shapeName=dyn_hair_back_01_06_r_capsule
   center=<3.9572 -0.3696 -3.8474>
   rotation=<rot 90 -6.5994 0>
  }]
 dyn_hair_back_01_05_l [{
   shapeType=Capsule
   length=0.4486
   radius=6.4171
   shapeName=dyn_hair_back_01_06_r_capsule
   center=<3.771 -0.1405 -3.8432>
   rotation=<rot 84.1446 3.9862 -0.0003>
  }]
 dyn_hair_back_01_05_r [{
   shapeType=Capsule
   length=0.4486
   radius=6.4171
   shapeName=dyn_hair_back_01_06_r_capsule
   center=<3.9391 -0.8862 -3.8474>
   rotation=<rot 90 -3.9338 0>
  }]
 dyn_hair_back_01_06_l [{
   shapeType=Capsule
   length=3.7484
   radius=7.1763
   shapeName=dyn_hair_back_01_06_r_capsule
   center=<6.3606 0.4018 -3.9427>
   rotation=<rot 90 0.9932 0>
  }]
 dyn_hair_back_01_06_r [{
   shapeType=Capsule
   length=3.7484
   radius=7.1763
   shapeName=dyn_hair_back_01_06_r_capsule
   center=<5.9614 -2.3819 -3.9427>
   rotation=<rot 90 -8.5929 0>
  }]
 dyn_hair_back_02_01_l [{shapeType=Capsule, length=0.4486, radius=6.4171, shapeName=dyn_hair_back_01_06_r_capsule, center=<4.4562 0 -4.7635>, rotation=<rot 90 0 0>}]
 dyn_hair_back_02_01_r [{shapeType=Capsule, length=0.4486, radius=6.4171, shapeName=dyn_hair_back_01_06_r_capsule, center=<4.4562 0 -4.7635>, rotation=<rot 90 0 0>}]
 dyn_hair_back_02_02_l [{shapeType=Capsule, length=0.4486, radius=6.4171, shapeName=dyn_hair_back_01_06_r_capsule, center=<4.4562 0 -4.3647>, rotation=<rot 90 0 0>}]
 dyn_hair_back_02_02_r [{shapeType=Capsule, length=0.4486, radius=6.4171, shapeName=dyn_hair_back_01_06_r_capsule, center=<4.4562 0 -4.3647>, rotation=<rot 90 0 0>}]
 dyn_hair_back_02_03_l [{shapeType=Capsule, length=0.4486, radius=6.4171, shapeName=dyn_hair_back_01_06_r_capsule, center=<4.4562 0 -4.0836>, rotation=<rot 90 0 0>}]
 dyn_hair_back_02_03_r [{shapeType=Capsule, length=0.4486, radius=6.4171, shapeName=dyn_hair_back_01_06_r_capsule, center=<4.4562 0 -4.0836>, rotation=<rot 90 0 0>}]
 dyn_hair_back_02_04_l [{shapeType=Capsule, length=0.4486, radius=6.4171, shapeName=dyn_hair_back_01_06_r_capsule, center=<4.4562 0 -3.8447>, rotation=<rot 90 0 0>}]
 dyn_hair_back_02_04_r [{shapeType=Capsule, length=0.4486, radius=6.4171, shapeName=dyn_hair_back_01_06_r_capsule, center=<4.4562 0 -3.8447>, rotation=<rot 90 0 0>}]
 dyn_hair_back_02_05_l [{shapeType=Capsule, length=0.4486, radius=6.4171, shapeName=dyn_hair_back_01_06_r_capsule, center=<4.4562 0 -4.0836>, rotation=<rot 90 0 0>}]
 dyn_hair_back_02_05_r [{shapeType=Capsule, length=0.4486, radius=6.4171, shapeName=dyn_hair_back_01_06_r_capsule, center=<4.4562 0 -4.0836>, rotation=<rot 90 0 0>}]
 dyn_hair_back_02_06_l [{shapeType=Capsule, length=2.15, radius=6.7859, shapeName=dyn_hair_back_01_06_r_capsule, center=<5.1865 0 -4.0407>, rotation=<rot 90 0 0>}]
 dyn_hair_back_02_06_r [{shapeType=Capsule, length=2.15, radius=6.7859, shapeName=dyn_hair_back_01_06_r_capsule, center=<5.1865 0 -4.0407>, rotation=<rot 90 0 0>}]
 spine_01 [{shapeType=Capsule, length=3.7476, radius=16.573, shapeName=spine_05_capsule, center=<3.2057 -2.6502 0>}]
 spine_02 [{shapeType=Capsule, length=3.7476, radius=16.573, shapeName=spine_05_capsule, center=<3.2057 -1.6162 0>}]
 spine_03 [{shapeType=Capsule, length=3.7476, radius=16.573, shapeName=spine_05_capsule, center=<3.2057 -1.1177 0>}]
 spine_04 [{shapeType=Capsule, length=3.7476, radius=16.573, shapeName=spine_05_capsule, center=<3.2057 0.0012 0>}]
 spine_05 [{shapeType=Capsule, length=3.7476, radius=16.573, shapeName=spine_05_capsule, center=<3.2057 2.3298 0>}]
 upperarm_l [{shapeType=Capsule, length=24.8812, radius=5, shapeName=lowerarm_l_capsule, center=<12.8481 -0.3499 0>, rotation=<rot 90 179.1482 0>}]
 upperarm_r [{shapeType=Capsule, length=24.8812, radius=5, shapeName=lowerarm_r_capsule, center=<-12.8481 0.3499 0>, rotation=<rot 90 -0.8518 0>}]
}
constraints [
 {bone1Name=dyn_hair_back_01_01_l, bone2Name=spine_05, swing1LimitDegrees=30, swing2LimitDegrees=35, twistLimitDegrees=5}
 {bone1Name=dyn_hair_back_01_02_l, bone2Name=dyn_hair_back_01_01_l, swing1LimitDegrees=25, swing2LimitDegrees=30, twistLimitDegrees=15}
 {bone1Name=dyn_hair_back_01_03_l, bone2Name=dyn_hair_back_01_02_l, swing1LimitDegrees=25, swing2LimitDegrees=25, twistLimitDegrees=15}
 {bone1Name=dyn_hair_back_01_04_l, bone2Name=dyn_hair_back_01_03_l, swing1LimitDegrees=35, swing2LimitDegrees=35, twistLimitDegrees=15}
 {bone1Name=dyn_hair_back_01_05_l, bone2Name=dyn_hair_back_01_04_l, swing1LimitDegrees=35, swing2LimitDegrees=35, twistLimitDegrees=15}
 {bone1Name=dyn_hair_back_01_06_l, bone2Name=dyn_hair_back_01_05_l, swing1LimitDegrees=35, swing2LimitDegrees=50, twistLimitDegrees=15}
 {bone1Name=dyn_hair_back_01_01_r, bone2Name=spine_05, swing1LimitDegrees=30, swing2LimitDegrees=35, twistLimitDegrees=5}
 {bone1Name=dyn_hair_back_01_02_r, bone2Name=dyn_hair_back_01_01_r, swing1LimitDegrees=25, swing2LimitDegrees=30, twistLimitDegrees=15}
 {bone1Name=dyn_hair_back_01_03_r, bone2Name=dyn_hair_back_01_02_r, swing1LimitDegrees=25, swing2LimitDegrees=25, twistLimitDegrees=15}
 {bone1Name=dyn_hair_back_01_04_r, bone2Name=dyn_hair_back_01_03_r, swing1LimitDegrees=35, swing2LimitDegrees=35, twistLimitDegrees=15}
 {bone1Name=dyn_hair_back_01_05_r, bone2Name=dyn_hair_back_01_04_r, swing1LimitDegrees=35, swing2LimitDegrees=35, twistLimitDegrees=15}
 {bone1Name=dyn_hair_back_01_06_r, bone2Name=dyn_hair_back_01_05_r, swing1LimitDegrees=35, swing2LimitDegrees=50, twistLimitDegrees=15}
 {bone1Name=dyn_hair_back_02_01_l, bone2Name=spine_05, swing1LimitDegrees=25, swing2LimitDegrees=25, twistLimitDegrees=5}
 {bone1Name=dyn_hair_back_02_02_l, bone2Name=dyn_hair_back_02_01_l, swing1LimitDegrees=25, swing2LimitDegrees=25, twistLimitDegrees=15}
 {bone1Name=dyn_hair_back_02_03_l, bone2Name=dyn_hair_back_02_02_l, swing1LimitDegrees=15, swing2LimitDegrees=25, twistLimitDegrees=15}
 {bone1Name=dyn_hair_back_02_04_l, bone2Name=dyn_hair_back_02_03_l, swing1LimitDegrees=25, swing2LimitDegrees=35, twistLimitDegrees=15}
 {bone1Name=dyn_hair_back_02_05_l, bone2Name=dyn_hair_back_02_04_l, swing1LimitDegrees=35, swing2LimitDegrees=35, twistLimitDegrees=15}
 {bone1Name=dyn_hair_back_02_06_l, bone2Name=dyn_hair_back_02_05_l, swing1LimitDegrees=35, swing2LimitDegrees=50, twistLimitDegrees=15}
 {bone1Name=dyn_hair_back_02_01_r, bone2Name=spine_05, swing1LimitDegrees=25, swing2LimitDegrees=25, twistLimitDegrees=5}
 {bone1Name=dyn_hair_back_02_03_r, bone2Name=dyn_hair_back_02_02_r, swing1LimitDegrees=15, swing2LimitDegrees=25, twistLimitDegrees=15}
 {bone1Name=dyn_hair_back_02_04_r, bone2Name=dyn_hair_back_02_03_r, swing1LimitDegrees=25, swing2LimitDegrees=35, twistLimitDegrees=15}
 {bone1Name=dyn_hair_back_02_05_r, bone2Name=dyn_hair_back_02_04_r, swing1LimitDegrees=35, swing2LimitDegrees=35, twistLimitDegrees=15}
 {bone1Name=dyn_hair_back_02_06_r, bone2Name=dyn_hair_back_02_05_r, swing1LimitDegrees=35, swing2LimitDegrees=50, twistLimitDegrees=15}
 {
  bone1Name=spine_03
  bone2Name=spine_02
  swing1LimitDegrees=45
  swing1Motion=Free
  swing2LimitDegrees=45
  swing2Motion=Free
  twistLimitDegrees=45
  twistMotion=Free
 }
 {
  bone1Name=spine_04
  bone2Name=spine_03
  swing1LimitDegrees=45
  swing1Motion=Free
  swing2LimitDegrees=45
  swing2Motion=Free
  twistLimitDegrees=45
  twistMotion=Free
 }
 {
  bone1Name=spine_05
  bone2Name=spine_04
  swing1LimitDegrees=45
  swing1Motion=Free
  swing2LimitDegrees=45
  swing2Motion=Free
  twistLimitDegrees=45
  twistMotion=Free
 }
 {bone1Name=dyn_hair_back_02_05_r, bone2Name=dyn_hair_back_01_05_r, swing1LimitDegrees=35, swing2LimitDegrees=35, twistLimitDegrees=35}
 {bone1Name=dyn_hair_back_02_05_l, bone2Name=dyn_hair_back_01_05_l, swing1LimitDegrees=35, swing2LimitDegrees=35, twistLimitDegrees=35}
 {bone1Name=dyn_hair_back_02_05_l, bone2Name=dyn_hair_back_02_05_r, swing1LimitDegrees=35, swing2LimitDegrees=35, twistLimitDegrees=35}
 {bone1Name=dyn_hair_back_02_06_r, bone2Name=dyn_hair_back_01_06_r, swing1LimitDegrees=35, swing2LimitDegrees=35, twistLimitDegrees=35}
 {bone1Name=dyn_hair_back_02_06_l, bone2Name=dyn_hair_back_01_06_l, swing1LimitDegrees=35, swing2LimitDegrees=35, twistLimitDegrees=35}
 {
  bone1Name=upperarm_l
  bone2Name=spine_05
  swing1LimitDegrees=45
  swing1Motion=Free
  swing2LimitDegrees=45
  swing2Motion=Free
  twistLimitDegrees=45
  twistMotion=Free
 }
 {
  bone1Name=upperarm_r
  bone2Name=spine_05
  swing1LimitDegrees=45
  swing1Motion=Free
  swing2LimitDegrees=45
  swing2Motion=Free
  twistLimitDegrees=45
  twistMotion=Free
 }
 {bone1Name=dyn_hair_back_02_06_l, bone2Name=dyn_hair_back_02_06_r, swing1LimitDegrees=35, swing2LimitDegrees=35, twistLimitDegrees=35}
 {bone1Name=dyn_hair_back_02_04_l, bone2Name=dyn_hair_back_02_04_r, swing1LimitDegrees=35, swing2LimitDegrees=35, twistLimitDegrees=35}
 {bone1Name=dyn_hair_back_02_03_r, bone2Name=dyn_hair_back_01_03_r, swing1LimitDegrees=35, swing2LimitDegrees=35, twistLimitDegrees=35}
 {bone1Name=dyn_hair_back_02_03_l, bone2Name=dyn_hair_back_01_03_l, swing1LimitDegrees=35, swing2LimitDegrees=35, twistLimitDegrees=35}
 {bone1Name=spine_05, bone2Name=dyn_hair_back_02_03_r, swing1LimitDegrees=45, swing2LimitDegrees=45, twistLimitDegrees=45}
 {bone1Name=spine_05, bone2Name=dyn_hair_back_02_03_l, swing1LimitDegrees=45, swing2LimitDegrees=45, twistLimitDegrees=45}
 {bone1Name=spine_05, bone2Name=dyn_hair_back_01_03_r, swing1LimitDegrees=45, swing2LimitDegrees=45, twistLimitDegrees=45}
 {bone1Name=spine_05, bone2Name=dyn_hair_back_01_02_r, swing1LimitDegrees=45, swing2LimitDegrees=45, twistLimitDegrees=45}
 {bone1Name=spine_05, bone2Name=dyn_hair_back_01_03_l, swing1LimitDegrees=45, swing2LimitDegrees=45, twistLimitDegrees=45}
 {bone1Name=dyn_hair_back_02_01_l, bone2Name=dyn_hair_back_02_01_r, swing1LimitDegrees=25, swing2LimitDegrees=25, twistLimitDegrees=25}
 {bone1Name=dyn_hair_back_02_01_l, bone2Name=dyn_hair_back_01_01_l, swing1LimitDegrees=35, swing2LimitDegrees=35, twistLimitDegrees=35}
 {bone1Name=dyn_hair_back_02_01_r, bone2Name=dyn_hair_back_01_01_r, swing1LimitDegrees=35, swing2LimitDegrees=35, twistLimitDegrees=35}
 {bone1Name=spine_05, bone2Name=dyn_hair_back_01_02_l, swing1LimitDegrees=45, swing2LimitDegrees=45, twistLimitDegrees=45}
 {bone1Name=dyn_hair_back_02_02_r, bone2Name=dyn_hair_back_02_01_r, swing1LimitDegrees=25, swing2LimitDegrees=25, twistLimitDegrees=15}
 {bone1Name=spine_05, bone2Name=dyn_hair_back_02_02_l, swing1LimitDegrees=45, swing2LimitDegrees=45, twistLimitDegrees=45}
 {bone1Name=spine_05, bone2Name=dyn_hair_back_02_02_r, swing1LimitDegrees=45, swing2LimitDegrees=45, twistLimitDegrees=45}
 {
  bone1Name=spine_02
  bone2Name=spine_01
  swing1LimitDegrees=45
  swing1Motion=Free
  swing2LimitDegrees=45
  swing2Motion=Free
  twistLimitDegrees=45
  twistMotion=Free
 }
]
